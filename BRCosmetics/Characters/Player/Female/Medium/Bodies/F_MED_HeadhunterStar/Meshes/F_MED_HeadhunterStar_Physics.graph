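# Physics asset for the HeadhunterStar medium female body cosmetic, setting up physics for its left and right strap bone chains.
# Capsules cover the pelvis, thighs, calves and every strap bone.
# On each side, a main and a back chain hang from the pelvis through free-rotating master bones, with mostly tight angular limits along the segments.
# Both chains join at the last back bone with free-rotating constraints, closing a loop.
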PhysicsAsset /BRCosmetics/Characters/Player/Female/Medium/Bodies/F_MED_HeadhunterStar/Meshes/F_MED_HeadhunterStar_Physics
bodies {
 calf_l [{shapeType=Capsule, length=37.3841, radius=8.7491, center=<-24.8098 -2.7726 -0.0783>, rotation=<rot 0 93.318 -87.9056>}]
 calf_r [{shapeType=Capsule, length=37.3841, radius=8.7491, center=<24.8098 2.7726 -0.0783>, rotation=<rot 0 93.3266 -88.5225>}]
 dyn_strap_01_l [{
   shapeType=Capsule
   length=1.6124
   radius=1.2594
   shapeName=dyn_strap_01_l_capsule
   center=<2.5171 -0.0972 0.0178>
   rotation=<rot 0 -176.1439 -70.6192>
  }]
 dyn_strap_01_r [{
   shapeType=Capsule
   length=1.6124
   radius=1.2594
   shapeName=dyn_strap_01_r_capsule
   center=<2.5169 0.0973 0.0181>
   rotation=<rot 0 -3.8561 -70.6194>
  }]
 dyn_strap_02_l [{shapeType=Capsule, length=1, radius=2, shapeName=dyn_strap_02_l_capsule, center=<2 0 0>, rotation=<rot 90 -49.2934 -49.2934>}]
 dyn_strap_02_r [{shapeType=Capsule, length=1, radius=2, shapeName=dyn_strap_02_l_capsule, center=<2 0 0>, rotation=<rot 90 0 0>}]
 dyn_strap_03_l [
  {shapeType=Capsule, length=1.5, radius=1.875, shapeName=dyn_strap_03_l_capsule, center=<2 0 0>, rotation=<rot 90 -16.753 -16.7529>}
  {
   shapeType=Capsule
   length=1.75
   radius=1.125
   shapeName=dyn_strap_03_l_capsule
   center=<2.2906 -0.3022 -1.8593>
   rotation=<rot 90 -160.5288 -160.5287>
  }
 ]
 dyn_strap_03_r [
  {shapeType=Capsule, length=1.5, radius=1.875, shapeName=dyn_strap_03_l_capsule, center=<2 0 0>, rotation=<rot 90 0 0>}
  {
   shapeType=Capsule
   length=1.875
   radius=1.25
   shapeName=dyn_strap_03_r_capsule
   center=<2.7471 0.2788 -2.2403>
   rotation=<rot 90 179.9604 179.9604>
  }
 ]
 dyn_strap_04_l [{shapeType=Capsule, length=7.5, radius=1.875, shapeName=dyn_strap_04_l_capsule, center=<5 0 0>, rotation=<rot 90 -13.154 -13.1539>}]
 dyn_strap_04_r [{shapeType=Capsule, length=7.5, radius=1.875, shapeName=dyn_strap_04_l_capsule, center=<5 0 0>, rotation=<rot 90 0 0>}]
 dyn_strap_05_l [{shapeType=Capsule, length=8.5, radius=2, shapeName=dyn_strap_05_l_capsule, center=<5 0 0>, rotation=<rot 90 -13.154 -13.1539>}]
 dyn_strap_05_r [{shapeType=Capsule, length=8.5, radius=2, shapeName=dyn_strap_05_l_capsule, center=<5 0 0>, rotation=<rot 90 0 0>}]
 dyn_strap_MSTR_l [{shapeType=Capsule, length=0.505, radius=0.505, shapeName=dyn_strap_MSTR_l_capsule, rotation=<rot 0 -90 0>}]
 dyn_strap_MSTR_r [{shapeType=Capsule, length=0.505, radius=0.505, shapeName=dyn_strap_MSTR_r_capsule, rotation=<rot 0 -90 0>}]
 dyn_strap_bk_01_MSTR_l [{shapeType=Capsule, length=0.505, radius=0.505, shapeName=dyn_strap_bk_01_MSTR_l_capsule, rotation=<rot 0 -90 0>}]
 dyn_strap_bk_01_MSTR_r [{shapeType=Capsule, length=0, radius=2.13, shapeName=dyn_strap_bk_01_MSTR_r_capsule, rotation=<rot 0 -90 0>}]
 dyn_strap_bk_01_l [{shapeType=Capsule, length=1.25, radius=2, shapeName=dyn_strap_bk_01_l_capsule, center=<2 0 0>, rotation=<rot 90 0 0>}]
 dyn_strap_bk_01_r [{shapeType=Capsule, length=1.25, radius=2, shapeName=dyn_strap_bk_01_l_capsule, center=<2 0 0>, rotation=<rot 90 0 0>}]
 dyn_strap_bk_02_l [{shapeType=Capsule, length=5.625, radius=1.875, shapeName=dyn_strap_bk_02_l_capsule, center=<3 0 0>, rotation=<rot 90 0 0>}]
 dyn_strap_bk_02_r [{shapeType=Capsule, length=5.625, radius=1.875, shapeName=dyn_strap_bk_02_l_capsule, center=<3 0 0>, rotation=<rot 90 0 0>}]
 dyn_strap_bk_03_l [{shapeType=Capsule, length=8.375, radius=1.875, shapeName=dyn_strap_bk_03_l_capsule, center=<6 0 0>, rotation=<rot 90 0 0>}]
 dyn_strap_bk_03_r [{shapeType=Capsule, length=8.375, radius=1.875, shapeName=dyn_strap_bk_03_l_capsule, center=<6 0 0>, rotation=<rot 90 0 0>}]
 dyn_strap_bk_04_l [{shapeType=Capsule, length=0.505, radius=0.505, shapeName=dyn_strap_bk_04_l_capsule, rotation=<rot 0 -90 0>}]
 dyn_strap_bk_04_r [{shapeType=Capsule, length=0.505, radius=0.505, shapeName=dyn_strap_bk_04_r_capsule, rotation=<rot 0 -90 0>}]
 pelvis [{shapeType=Capsule, length=10.6376, radius=8.4487, center=<-5.8341 -1.5843 0>, rotation=<rot 0 -0.0001 0>}]
 thigh_l [{shapeType=Capsule, length=23.9504, radius=8.2666, center=<-18.1586 -0.1692 1.8694>, rotation=<rot 0 -90.9608 -96.1276>}]
 thigh_r [{shapeType=Capsule, length=23.9504, radius=8.2666, center=<18.9015 0.283 -1.7902>, rotation=<rot 0 -90.9609 -96.1276>}]
}
constraints [
 {bone1Name=thigh_l, bone2Name=pelvis, swing1LimitDegrees=45, swing2LimitDegrees=45, twistLimitDegrees=45}
 {bone1Name=calf_l, bone2Name=thigh_l, swing1LimitDegrees=45, swing2LimitDegrees=45, twistLimitDegrees=45}
 {bone1Name=thigh_r, bone2Name=pelvis, swing1LimitDegrees=45, swing2LimitDegrees=45, twistLimitDegrees=45}
 {bone1Name=calf_r, bone2Name=thigh_r, swing1LimitDegrees=45, swing2LimitDegrees=45, twistLimitDegrees=45}
 {
  bone1Name=dyn_strap_MSTR_r
  bone2Name=pelvis
  swing1LimitDegrees=45
  swing1Motion=Free
  swing2LimitDegrees=45
  swing2Motion=Free
  twistLimitDegrees=45
  twistMotion=Free
 }
 {
  bone1Name=dyn_strap_01_r
  bone2Name=dyn_strap_MSTR_r
  swing1LimitDegrees=45
  swing1Motion=Free
  swing2LimitDegrees=45
  swing2Motion=Free
  twistLimitDegrees=45
  twistMotion=Free
 }
 {bone1Name=dyn_strap_02_r, bone2Name=dyn_strap_01_r, swing1LimitDegrees=10, swing2LimitDegrees=25, twistLimitDegrees=5}
 {bone1Name=dyn_strap_03_r, bone2Name=dyn_strap_02_r, swing1LimitDegrees=35, swing2LimitDegrees=15, twistLimitDegrees=15}
 {bone1Name=dyn_strap_04_r, bone2Name=dyn_strap_03_r, swing1LimitDegrees=35, swing2LimitDegrees=15, twistLimitDegrees=5}
 {bone1Name=dyn_strap_05_r, bone2Name=dyn_strap_04_r, swing1LimitDegrees=35, swing2LimitDegrees=15, twistLimitDegrees=5}
 {
  bone1Name=dyn_strap_bk_01_MSTR_l
  bone2Name=pelvis
  swing1LimitDegrees=45
  swing1Motion=Free
  swing2LimitDegrees=45
  swing2Motion=Free
  twistLimitDegrees=45
  twistMotion=Free
 }
 {bone1Name=dyn_strap_bk_01_l, bone2Name=dyn_strap_bk_01_MSTR_l, swing1LimitDegrees=10, swing2LimitDegrees=35, twistLimitDegrees=5}
 {bone1Name=dyn_strap_bk_02_l, bone2Name=dyn_strap_bk_01_l, swing1LimitDegrees=10, swing2LimitDegrees=35, twistLimitDegrees=5}
 {bone1Name=dyn_strap_bk_03_l, bone2Name=dyn_strap_bk_02_l, swing1LimitDegrees=10, swing2LimitDegrees=35, twistLimitDegrees=5}
 {
  bone1Name=dyn_strap_bk_01_MSTR_r
  bone2Name=pelvis
  swing1LimitDegrees=45
  swing1Motion=Free
  swing2LimitDegrees=45
  swing2Motion=Free
  twistLimitDegrees=45
  twistMotion=Free
 }
 {bone1Name=dyn_strap_bk_01_r, bone2Name=dyn_strap_bk_01_MSTR_r, swing1LimitDegrees=10, swing2LimitDegrees=35, twistLimitDegrees=5}
 {bone1Name=dyn_strap_bk_02_r, bone2Name=dyn_strap_bk_01_r, swing1LimitDegrees=10, swing2LimitDegrees=35, twistLimitDegrees=5}
 {bone1Name=dyn_strap_bk_03_r, bone2Name=dyn_strap_bk_02_r, swing1LimitDegrees=10, swing2LimitDegrees=35, twistLimitDegrees=5}
 {bone1Name=dyn_strap_05_l, bone2Name=dyn_strap_04_l, swing1LimitDegrees=35, swing2LimitDegrees=15, twistLimitDegrees=5}
 {
  bone1Name=dyn_strap_MSTR_l
  bone2Name=pelvis
  swing1LimitDegrees=45
  swing1Motion=Free
  swing2LimitDegrees=45
  swing2Motion=Free
  twistLimitDegrees=45
  twistMotion=Free
 }
 {
  bone1Name=dyn_strap_01_l
  bone2Name=dyn_strap_MSTR_l
  swing1LimitDegrees=45
  swing1Motion=Free
  swing2LimitDegrees=45
  swing2Motion=Free
  twistLimitDegrees=45
  twistMotion=Free
 }
 {bone1Name=dyn_strap_02_l, bone2Name=dyn_strap_01_l, swing1LimitDegrees=10, swing2LimitDegrees=25, twistLimitDegrees=5}
 {bone1Name=dyn_strap_03_l, bone2Name=dyn_strap_02_l, swing1LimitDegrees=35, swing2LimitDegrees=15, twistLimitDegrees=15}
 {bone1Name=dyn_strap_04_l, bone2Name=dyn_strap_03_l, swing1LimitDegrees=35, swing2LimitDegrees=15, twistLimitDegrees=5}
 {
  bone1Name=dyn_strap_bk_04_l
  bone2Name=dyn_strap_bk_03_l
  swing1LimitDegrees=45
  swing1Motion=Free
  swing2LimitDegrees=45
  swing2Motion=Free
  twistLimitDegrees=45
  twistMotion=Free
 }
 {
  bone1Name=dyn_strap_bk_04_r
  bone2Name=dyn_strap_bk_03_r
  swing1LimitDegrees=45
  swing1Motion=Free
  swing2LimitDegrees=45
  swing2Motion=Free
  twistLimitDegrees=45
  twistMotion=Free
 }
 {
  bone1Name=dyn_strap_bk_04_r
  bone2Name=dyn_strap_05_r
  swing1LimitDegrees=45
  swing1Motion=Free
  swing2LimitDegrees=45
  swing2Motion=Free
  twistLimitDegrees=45
  twistMotion=Free
 }
 {
  bone1Name=dyn_strap_bk_04_l
  bone2Name=dyn_strap_05_l
  swing1LimitDegrees=45
  swing1Motion=Free
  swing2LimitDegrees=45
  swing2Motion=Free
  twistLimitDegrees=45
  twistMotion=Free
 }
]
