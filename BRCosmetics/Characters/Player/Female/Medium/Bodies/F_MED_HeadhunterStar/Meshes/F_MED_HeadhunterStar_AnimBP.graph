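# Cosmetic AnimBP for the F_MED_HeadhunterStar player body, adding secondary motion to straps, kneepads and shirt.
# The copied parent pose feeds a deform Control Rig, weighted bone constraints, a second Control Rig and additive strap offsets for sliding, front end and jumping/falling.
# RigidBody nodes swap on skydiving, then AnimDynamics springs the shirt.
# The EventGraph updates gravity override and, outside front end, drives material WPO cloth from leg motion and skydive-float wind.
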AnimBlueprint /BRCosmetics/Characters/Player/Female/Medium/Bodies/F_MED_HeadhunterStar/Meshes/F_MED_HeadhunterStar_AnimBP
target_skeleton=&/F_MED_HeadhunterStar_Skeleton
graphs [
 AnimGraph {nodes [
   n1 = Control Rig(Source=Copy Pose From Mesh() {bUseAttachedParent=true}) {
    alphaBoolBlend {blendOption=Linear}
    controlRigAssetReference {blueprintRigClass=&/F_MED_HeadhunterStar_DeformRig_ControlRig.F_MED_HeadhunterStar_DeformRig_ControlRig_C}
   }
   n2 = "Constraint - dyn_strap_MSTR_l (PPPPPP)"!(
    ComponentPose=Local To Component(LocalPose=@n1)
    Alpha=1
    ConstraintWeights_0=0.0316
    ConstraintWeights_1:float=0.0004
    ConstraintWeights_2:float=0.8827
    ConstraintWeights_3:float=0.0659
    ConstraintWeights_4:float=0.0126
    ConstraintWeights_5:float=0.0068
   ) -> (Pose) {
    alphaBoolBlend {blendOption=Linear}
    boneToModify {boneName=dyn_strap_MSTR_l}
    constraintSetup [
     {targetBone {boneName=spine_02}}
     {targetBone {boneName=spine_01}}
     {targetBone {boneName=pelvis}}
     {targetBone {boneName=thigh_twist_01_l}}
     {targetBone {boneName=deform_glute_l}}
     {targetBone {boneName=deform_groin_l}}
    ]
    constraintWeights [1,1,1,1,1,1]
   }
   n3 = "Constraint - dyn_strap_MSTR_r (PPPPPP)"!(
    ComponentPose=@n2
    Alpha=1
    ConstraintWeights_0=0.0316
    ConstraintWeights_1:float=0.0004
    ConstraintWeights_2:float=0.8827
    ConstraintWeights_3:float=0.0659
    ConstraintWeights_4:float=0.0126
    ConstraintWeights_5:float=0.0068
   ) -> (Pose) {
    alphaBoolBlend {blendOption=Linear}
    boneToModify {boneName=dyn_strap_MSTR_r}
    constraintSetup [
     {targetBone {boneName=spine_02}}
     {targetBone {boneName=spine_01}}
     {targetBone {boneName=pelvis}}
     {targetBone {boneName=thigh_twist_01_r}}
     {targetBone {boneName=deform_glute_r}}
     {targetBone {boneName=deform_groin_r}}
    ]
    constraintWeights [1,1,1,1,1,1]
   }
   n4 = "Constraint - dyn_strap_bk_01_MSTR_l (PPPPPP)"!(
    ComponentPose=@n3
    Alpha=1
    ConstraintWeights_0=0.003
    ConstraintWeights_1:float=0.8934
    ConstraintWeights_2:float=0.0003
    ConstraintWeights_3:float=0.0003
    ConstraintWeights_4:float=0.0513
    ConstraintWeights_5:float=0.0516
   ) -> (Pose) {
    alphaBoolBlend {blendOption=Linear}
    boneToModify {boneName=dyn_strap_bk_01_MSTR_l}
    constraintSetup [
     {targetBone {boneName=spine_02}}
     {targetBone {boneName=pelvis}}
     {targetBone {boneName=thigh_twist_01_l}}
     {targetBone {boneName=thigh_twist_01_r}}
     {targetBone {boneName=deform_glute_l}}
     {targetBone {boneName=deform_glute_r}}
    ]
    constraintWeights [1,1,1,1,1,1]
   }
   n5 = "Constraint - dyn_strap_bk_01_MSTR_r (PPPPPP)"!(
    ComponentPose=@n4
    Alpha=1
    ConstraintWeights_0=0.003
    ConstraintWeights_1:float=0.8934
    ConstraintWeights_2:float=0.0003
    ConstraintWeights_3:float=0.0003
    ConstraintWeights_4:float=0.0513
    ConstraintWeights_5:float=0.0516
   ) -> (Pose) {
    alphaBoolBlend {blendOption=Linear}
    boneToModify {boneName=dyn_strap_bk_01_MSTR_r}
    constraintSetup [
     {targetBone {boneName=spine_02}}
     {targetBone {boneName=pelvis}}
     {targetBone {boneName=thigh_twist_01_l}}
     {targetBone {boneName=thigh_twist_01_r}}
     {targetBone {boneName=deform_glute_l}}
     {targetBone {boneName=deform_glute_r}}
    ]
    constraintWeights [1,1,1,1,1,1]
   }
   n6 = "Constraint - dyn_kneepad_l (RR)"!(ComponentPose=@n5, Alpha=1, ConstraintWeights_0=0.4, ConstraintWeights_1:float=0.6) -> (Pose) {
    alphaBoolBlend {blendOption=Linear}
    boneToModify {boneName=dyn_kneepad_l}
    constraintSetup [{transformType=Rotation, targetBone {boneName=calf_l}}, {transformType=Rotation, targetBone {boneName=thigh_l}}]
    constraintWeights [1,1]
   }
   n7 = "Constraint - dyn_kneepad_r (RR)"!(ComponentPose=@n6, Alpha=1, ConstraintWeights_0=0.4, ConstraintWeights_1:float=0.6) -> (Pose) {
    alphaBoolBlend {blendOption=Linear}
    boneToModify {boneName=dyn_kneepad_r}
    constraintSetup [{transformType=Rotation, targetBone {boneName=calf_r}}, {transformType=Rotation, targetBone {boneName=thigh_r}}]
    constraintWeights [1,1]
   }
   n8 = Control Rig!(Source=Component To Local(ComponentPose=@n7), Alpha, rban:bool=Get bRigidBodyEnabled()) -> (Pose) {
    alphaBoolBlend {blendOption=Linear}
    controlRigAssetReference {blueprintRigClass=&/F_MED_HeadhunterStar_CtrlRig.F_MED_HeadhunterStar_CtrlRig_C}
   }
   n9 = Get bIsSliding()
   n10 = "Transform (Modify) Bone - Bone: dyn_strap_bk_01_MSTR_l"!(ComponentPose=Local To Component(LocalPose=@n8), bAlphaBoolEnabled:bool=@n9, Translation="0,0,0", Rotation="-150,50,0", Scale="1,1,1") -> (Pose) {
    alphaInputType=Bool
    rotationMode=BMM_Additive
    rotationSpace=BCS_BoneSpace
    alphaBoolBlend {blendInTime=0.1, blendOption=Linear, blendOutTime=0.1}
    boneToModify {boneName=dyn_strap_bk_01_MSTR_l}
    rotation {pitch=-150, yaw=50}
   }
   n11 = "Transform (Modify) Bone - Bone: dyn_strap_bk_01_MSTR_r"!(ComponentPose=@n10, bAlphaBoolEnabled:bool=@n9, Translation="0,0,0", Rotation="-120,-50,0", Scale="1,1,1") -> (Pose) {
    alphaInputType=Bool
    rotationMode=BMM_Additive
    rotationSpace=BCS_BoneSpace
    alphaBoolBlend {blendInTime=0.1, blendOption=Linear, blendOutTime=0.1}
    boneToModify {boneName=dyn_strap_bk_01_MSTR_r}
    rotation {pitch=-120, yaw=-50}
   }
   n12 = "Transform (Modify) Bone - Bone: dyn_strap_MSTR_l"!(ComponentPose=@n11, bAlphaBoolEnabled:bool=@n9, Translation="0,0,0", Rotation="-30,70,30", Scale="1,1,1") -> (Pose) {
    alphaInputType=Bool
    rotationMode=BMM_Additive
    rotationSpace=BCS_BoneSpace
    alphaBoolBlend {blendInTime=0.1, blendOption=Linear, blendOutTime=0.1}
    boneToModify {boneName=dyn_strap_MSTR_l}
    rotation=<rot -30 70 30>
   }
   n13 = "Transform (Modify) Bone - Bone: dyn_strap_MSTR_r"!(ComponentPose=@n12, bAlphaBoolEnabled:bool=@n9, Translation="0,0,0", Rotation="-30,-70,-30", Scale="1,1,1") -> (Pose) {
    alphaInputType=Bool
    rotationMode=BMM_Additive
    rotationSpace=BCS_BoneSpace
    alphaBoolBlend {blendInTime=0.1, blendOption=Linear, blendOutTime=0.1}
    boneToModify {boneName=dyn_strap_MSTR_r}
    rotation=<rot -30 -70 -30>
   }
   n14 = "Transform (Modify) Bone - Bone: dyn_strap_MSTR_l"(ComponentPose=@n13, bAlphaBoolEnabled=Get bIsInFrontEnd(), Translation="0,0,0", Rotation="-7,0,0", Scale="1,1,1") {
    alphaInputType=Bool
    rotationMode=BMM_Additive
    rotationSpace=BCS_BoneSpace
    alphaBoolBlend {blendInTime=0.1, blendOption=Linear, blendOutTime=0.1}
    boneToModify {boneName=dyn_strap_MSTR_l}
    rotation {pitch=-7}
   }
   n15 = Get jumping() -> (jumping:double)
   n16 = "Transform (Modify) Bone - Bone: dyn_strap_MSTR_l"!(ComponentPose=@n14, Alpha:float=@n15, Translation="0,0,0", Rotation="-15,25,-25", Scale="1,1,1") -> (Pose) {
    rotationMode=BMM_Additive
    rotationSpace=BCS_BoneSpace
    alphaBoolBlend {blendInTime=0.1, blendOption=Linear, blendOutTime=0.5}
    alphaScaleBiasClamp {bInterpResult=true}
    boneToModify {boneName=dyn_strap_MSTR_l}
    rotation=<rot -15 25 -25>
   }
   n17 = "Transform (Modify) Bone - Bone: dyn_strap_MSTR_r"!(ComponentPose=@n16, Alpha:float=@n15, Translation="0,0,0", Rotation="-15,-25,25", Scale="1,1,1") -> (Pose) {
    rotationMode=BMM_Additive
    rotationSpace=BCS_BoneSpace
    alphaBoolBlend {blendInTime=0.1, blendOption=Linear, blendOutTime=0.5}
    alphaScaleBiasClamp {bInterpResult=true}
    boneToModify {boneName=dyn_strap_MSTR_r}
    rotation=<rot -15 -25 25>
   }
   n18 = "Transform (Modify) Bone - Bone: dyn_strap_bk_01_MSTR_l"!(ComponentPose=@n17, Alpha:float=@n15, Translation="0,0,0", Rotation="-15,-15,0", Scale="1,1,1") -> (Pose) {
    rotationMode=BMM_Additive
    rotationSpace=BCS_BoneSpace
    alphaBoolBlend {blendInTime=0.1, blendOption=Linear, blendOutTime=0.5}
    alphaScaleBiasClamp {bInterpResult=true}
    boneToModify {boneName=dyn_strap_bk_01_MSTR_l}
    rotation {pitch=-15, yaw=-15}
   }
   n19 = "Transform (Modify) Bone - Bone: dyn_strap_bk_01_MSTR_r"!(ComponentPose=@n18, Alpha:float=@n15, Translation="0,0,0", Rotation="-15,15,0", Scale="1,1,1") -> (Pose) {
    rotationMode=BMM_Additive
    rotationSpace=BCS_BoneSpace
    alphaBoolBlend {blendInTime=0.1, blendOption=Linear, blendOutTime=0.5}
    alphaScaleBiasClamp {bInterpResult=true}
    boneToModify {boneName=dyn_strap_bk_01_MSTR_r}
    rotation {pitch=-15, yaw=15}
   }
   n20 = Get bIsSkyDiving()
   n21 = Get GravityOverride()
   n22 = RigidBody!(ComponentPose=@n19, bAlphaBoolEnabled:bool=NOT Boolean(A=@n20), OverrideWorldGravity:Vector=@n21, ExternalForce) -> (Pose) {
    alphaInputType=Bool
    bClampLinearTranslationLimitToRefPose=true
    bOverrideWorldGravity=true
    overridePhysicsAsset=&/F_MED_HeadhunterStar_Physics
    simulationSpace=BaseBoneSpace
    alphaBoolBlend {blendInTime=0.1, blendOption=Linear, blendOutTime=0.1}
    baseBoneRef {boneName=pelvis}
    componentAppliedLinearAccClamp=<5000 5000 5000>
    componentLinearAccScale=<0.3 0.4 0.4>
    componentLinearVelScale=<0.3 0.4 0.4>
    overrideWorldGravity {z=-980}
    simSpaceSettings {
     maxAngularAcceleration=1000
     maxAngularVelocity=3
     maxLinearAcceleration=700
     maxLinearVelocity=500
     velocityScaleZ=0
     worldAlpha=1
     externalLinearDragV=<0.2 0.2>
    }
   }
   n23 = RigidBody(ComponentPose=@n22, bAlphaBoolEnabled=@n20, OverrideWorldGravity=@n21) {
    alphaInputType=Bool
    bClampLinearTranslationLimitToRefPose=true
    bOverrideWorldGravity=true
    overridePhysicsAsset=&F_MED_HeadhunterStar_Physics
    simulationSpace=BaseBoneSpace
    alphaBoolBlend {blendInTime=0.1, blendOption=Linear, blendOutTime=0.1}
    baseBoneRef {boneName=root}
    componentAppliedLinearAccClamp=<5000 5000 5000>
    componentLinearAccScale=<0.3 0.3 0.5>
    componentLinearVelScale=<0.3 0.3 0.5>
    overrideWorldGravity {z=-980}
   }
   n24 = Get bIsSliding()
   n25 = "Transform (Modify) Bone - Bone: dyn_strap_bk_02_r"!(ComponentPose=@n23, bAlphaBoolEnabled:bool=@n24, Translation="0,0,0", Rotation="-20,25,-15", Scale="1,1,1") -> (Pose) {
    alphaInputType=Bool
    rotationMode=BMM_Additive
    rotationSpace=BCS_BoneSpace
    alphaBoolBlend {blendInTime=0.1, blendOption=Linear, blendOutTime=0.1}
    boneToModify {boneName=dyn_strap_bk_02_r}
    rotation=<rot -20 25 -15>
   }
   n26 = "Transform (Modify) Bone - Bone: dyn_strap_03_r"!(ComponentPose=@n25, bAlphaBoolEnabled:bool=@n24, Translation="0,0,0", Rotation="0,-20,0", Scale="1,1,1") -> (Pose) {
    alphaInputType=Bool
    rotationMode=BMM_Additive
    rotationSpace=BCS_BoneSpace
    alphaBoolBlend {blendInTime=0.1, blendOption=Linear, blendOutTime=0.1}
    boneToModify {boneName=dyn_strap_03_r}
    rotation {yaw=-20}
   }
   n27 = Anim Dynamics - Bone: dyn_shirt_l(ComponentPose=@n26, Alpha=1, ExternalForce="0,0,0") {
    angularDampingOverride=0.97
    angularSpringConstant=50
    bAngularSpring=true
    bGravityOverrideInSimSpace=true
    bOverrideAngularDamping=true
    simulationSpace=RootRelative
    alphaBoolBlend {blendOption=Linear}
    boundBone {boneName=dyn_shirt_l}
    componentAppliedLinearAccClamp=<3500 3500 3500>
    componentLinearAccScale=<1 1 1>
    componentLinearVelScale=<1 1 1>
    gravityOverride {z=-810}
    physicsBodyDefinitions [{
      sphereCollisionRadius=0
      boundBone {boneName=dyn_shirt_l}
      boxExtents=<20 20 10>
      constraintSetup {angularLimitsMax=<5 5 5>, angularLimitsMin=<-5 -5 -35>, angularTarget=<0 0 0>}
      localJointOffset=<3 0 0>
     }]
    relativeSpaceBone {boneName=root}
   }
   n28 = Anim Dynamics - Bone: dyn_shirt_r(ComponentPose=@n27, Alpha=1, ExternalForce="0,0,0") {
    angularDampingOverride=0.97
    angularSpringConstant=50
    bAngularSpring=true
    bGravityOverrideInSimSpace=true
    bOverrideAngularDamping=true
    simulationSpace=RootRelative
    alphaBoolBlend {blendOption=Linear}
    boundBone {boneName=dyn_shirt_r}
    componentAppliedLinearAccClamp=<3500 3500 3500>
    componentLinearAccScale=<1 1 1>
    componentLinearVelScale=<1 1 1>
    gravityOverride {z=-810}
    physicsBodyDefinitions [{
      sphereCollisionRadius=0
      boundBone {boneName=dyn_shirt_r}
      boxExtents=<20 20 10>
      constraintSetup {angularLimitsMax=<5 5 5>, angularLimitsMin=<-5 -5 -35>, angularTarget=<0 0 0>}
      localJointOffset=<3 0 0>
     }]
    relativeSpaceBone {boneName=root}
   }
   n29 = Anim Dynamics - Bone: dyn_shirt_bk(ComponentPose=@n28, Alpha=1, ExternalForce="0,0,0") {
    angularDampingOverride=0.97
    angularSpringConstant=50
    bAngularSpring=true
    bGravityOverrideInSimSpace=true
    bOverrideAngularDamping=true
    simulationSpace=RootRelative
    alphaBoolBlend {blendOption=Linear}
    boundBone {boneName=dyn_shirt_bk}
    componentAppliedLinearAccClamp=<3500 3500 3500>
    componentLinearAccScale=<1 1 1>
    componentLinearVelScale=<1 1 1>
    gravityOverride {z=-810}
    physicsBodyDefinitions [{
      sphereCollisionRadius=0
      boundBone {boneName=dyn_shirt_bk}
      boxExtents=<20 20 10>
      constraintSetup {angularLimitsMax=<5 5 5>, angularLimitsMin=<-5 -5 -35>, angularTarget=<0 0 0>}
      localJointOffset=<3 0 0>
     }]
    relativeSpaceBone {boneName=root}
   }
   Output Pose(Result=Component To Local(ComponentPose=@n29)) {layerGroup=None}
  ]}
 EventGraph {nodes [
   n1 = Event BlueprintUpdateAnimation()
   > n2 = Sequence!(execute) -> (then_0, then_1, then_2:Exec)
   > n3 = Gravity Override State Function DEPRECATED(
    CharacterPartAnimInstance=Self-Reference()
    GameTimeInSeconds=GetGameTimeInSeconds()
    GravityOverrideParams=Get GravityOverrideParams() -> (GravityOverrideParams:Gravity Override Params Struct Structure)
   )
   > Set GravityOverride(GravityOverride=@n3)
   > n4 = Branch(Condition=Get bRigidBodyEnabled())
   > Set rban(rban:double=1) -> (Output_Get:double)
   Set rban(execute=@n4.else, rban=0)
   Is Valid(exec=@n2.then_1, InputObject=Get PlayerPawn() -> (PlayerPawn:Fort Player Pawn Athena Object Reference))
   > n5 = Branch(Condition=Get bIsInFrontEnd())
   Set WPO Globals Function(
    execute=@n5.else
    Material=Get Mat Instance Dynamic()
    Overlay Material=Get Overlay Mat Instance Dynamic()
    Skeletal Mesh=Get SkelMesh()
    "Local Noise Transform  Location"=Get Local Noise Transform Location() -> (Local Noise Transform Location:Vector)
    Local Noise Offset=Get Local Noise Offset()
   )
   > Set WPO Regional Function!(
    execute
    Material:Material Instance Dynamic Object Reference=Get Mat Instance Dynamic()
    Overlay Material:Material Instance Dynamic Object Reference=Get Overlay Mat Instance Dynamic()
    Skeletal Mesh:Skeletal Mesh Component Object Reference=Get SkelMesh()
    Velocity Transform:Name=foot_r
    Velocity Interp Speed:double=Get Velocity Interp Speed()
    Collision Transform:Name=calf_r
    Collision Transform Axis:EAxis Enum
    Invert Collision Transform Axis:bool=false
    Delta Time X:double=@n1.DeltaTimeX
    Previous Velocity Transform Location:"Vector (by ref)"=Get R1 Velocity Transform Location()
    Previous Velocity Transform Velocity:"Vector (by ref)"=Get R1 Velocity Transform Velocity()
    Region:int=1
   ) -> (then)
   > Set WPO Regional Function(
    Material=Get Mat Instance Dynamic()
    Overlay Material=Get Overlay Mat Instance Dynamic()
    Skeletal Mesh=Get SkelMesh()
    Velocity Transform=foot_l
    Velocity Interp Speed=Get Velocity Interp Speed()
    Collision Transform=calf_l
    Invert Collision Transform Axis=true
    Delta Time X=@n1.DeltaTimeX
    Previous Velocity Transform Location=Get R2 Velocity Transform Location()
    Previous Velocity Transform Velocity=Get R2 Velocity Transform Velocity()
    Region=2
   )
   > n6 = Branch(Condition=Get bIsSkydiveFloating())
   > n7 = Set WPO Pho Wind Intensity(WPO Pho Wind Intensity=1)
   SetScalarParameterValue(execute=[@n8.then, @n7.then], self=Get Mat Instance Dynamic(), ParameterName=ClothSim_WindStrength, Value=Get WPO Pho Wind Intensity())
   n8 = Set WPO Pho Wind Intensity(execute=@n6.else, WPO Pho Wind Intensity=0)
   Reroute Node(InputPin=@n2.then_2)
   Event BlueprintInitializeAnimation()
   > Set SkelMesh(SkelMesh=GetOwnerSkeletalMeshComponent())
   n9 = TryGetPawnOwner()
   > Is Valid(InputObject=@n9)
   > n10 = Cast To FortPlayerPawnAthena(Object=@n9)
   > Set PlayerPawn(PlayerPawn:Fort Player Pawn Athena Object Reference=@n10) -> (Output_Get:Fort Player Pawn Athena Object Reference)
   n11 = Event OnBeginJumping()
   Set jumping(execute=[@n11.then, @n16.then], jumping:double=1) -> (Output_Get:double)
   n12 = Event OnEndJumping()
   Set jumping(execute=[@n12.then, @n17.then], jumping=0)
   Event OnCharacterCustomizationLoaded()
   > n13 = Set SkelMesh(SkelMesh=GetPartSkeletalMeshComponent())
   > n14 = Get Dynamic Material Instance(SkeletalMeshComponent=@n13)
   > Set Mat Instance Dynamic(Mat Instance Dynamic=@n14)
   > n15 = GetOrCreateDynamicOverlayMaterial(MeshComponent=@n13, MaterialName=None)
   > Set Overlay Mat Instance Dynamic(Overlay Mat Instance Dynamic=@n15)
   n16 = Event OnBeginFalling()
   n17 = Event OnEndFalling()
  ]}
]
variables [
 ClothSim_VelocityMax {type=number}
 ClothSim_WindStrength {type=number}
 GravityOverride {type=Vector}
 GravityOverrideParams {type=GravityOverrideParamsStruct}
 jumping {type=number}
 Local Noise Offset {type=Vector}
 Local Noise Transform Location {type=Vector}
 Mat Instance Dynamic {type=MaterialInstanceDynamic}
 movementSum {type=boolean}
 Overlay Mat Instance Dynamic {type=MaterialInstanceDynamic}
 PlayerPawn {type=FortPlayerPawnAthena}
 R1 Velocity Transform Location {type=Vector}
 R1 Velocity Transform Velocity {type=Vector}
 R2 Velocity Transform Location {type=Vector}
 R2 Velocity Transform Velocity {type=Vector}
 rban {type=number}
 skeletal_mesh {type=SkeletalMeshComponent}
 SkelMesh {type=SkeletalMeshComponent}
 Velocity Interp Speed {type=number}
 WPO Pho Wind Intensity {type=number}
]
