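# Physics asset for the face-accessory mesh of the female medium HeadhunterStar player cosmetic, covering hair, helmet sides and headphones.
# Capsule bodies span the upper spine, shoulders, upper arms, neck and head, plus the dyn_ accessory bones.
# Bangs and front-hair chains hang from MSTR bodies on the head, with limited swing and tight twist per link.
# Body, MSTR, helmet-side and headphone constraints leave swing and twist free; headphones attach to neck_01.
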PhysicsAsset /BRCosmetics/Characters/Player/Female/Medium/Bodies/F_MED_HeadhunterStar/Meshes/Parts/F_MED_HeadhunterStar_FR_FaceAcc_Physics
bodies {
 clavicle_l [{
   shapeType=Capsule
   length=4.7081
   radius=5.3397
   shapeName=clavicle_l_capsule
   center=<6.5243 -0.6953 0.104>
   rotation=<rot 16.9211 87.287 -90>
  }]
 clavicle_r [{
   shapeType=Capsule
   length=5.6159
   radius=5.6487
   shapeName=clavicle_r_capsule
   center=<-6.4341 0.3237 0.8599>
   rotation=<rot 3.2384 98.0696 -90>
  }]
 dyn_bangs_l_01 [{shapeType=Capsule, length=4.25, radius=0.75, shapeName=dyn_bangs_l_01_capsule, center=<2 0 0>, rotation=<rot 90 -97.076 -97.0763>}]
 dyn_bangs_l_02 [{shapeType=Capsule, length=4.25, radius=0.75, shapeName=dyn_bangs_l_02_capsule, center=<2 0 0>, rotation=<rot 90 -94.1205 -94.1207>}]
 dyn_bangs_l_03 [{shapeType=Capsule, length=4.25, radius=0.75, shapeName=dyn_bangs_l_03_capsule, center=<2 0 0>, rotation=<rot 90 -95.1519 -95.1521>}]
 dyn_bangs_l_04 [{shapeType=Capsule, length=4.25, radius=0.75, shapeName=dyn_bangs_l_04_capsule, center=<2 0 0>, rotation=<rot 90 86.3546 86.3544>}]
 dyn_bangs_l_MSTR [{shapeType=Capsule, length=0.505, radius=0.505, shapeName=dyn_bangs_l_MSTR_capsule, rotation=<rot 0 -90 0>}]
 dyn_bangs_r_01 [{shapeType=Capsule, length=4.25, radius=0.75, shapeName=dyn_bangs_r_01_capsule, center=<2 0 0>, rotation=<rot 90 85.8642 85.864>}]
 dyn_bangs_r_02 [{shapeType=Capsule, length=4.25, radius=0.75, shapeName=dyn_bangs_r_02_capsule, center=<2 0 0>, rotation=<rot 90 86.0638 86.0636>}]
 dyn_bangs_r_03 [{shapeType=Capsule, length=4.25, radius=0.75, shapeName=dyn_bangs_r_03_capsule, center=<2 0 0>, rotation=<rot 90 85.2229 85.2226>}]
 dyn_bangs_r_04 [{shapeType=Capsule, length=4.25, radius=0.75, shapeName=dyn_bangs_r_04_capsule, center=<2 0 0>, rotation=<rot 90 -93.6861 -93.6863>}]
 dyn_bangs_r_MSTR [{shapeType=Capsule, length=0.505, radius=0.505, shapeName=dyn_bangs_r_MSTR_capsule, rotation=<rot 0 -90 0>}]
 dyn_hair_front_l_01 [{shapeType=Capsule, length=3.125, radius=2.875, shapeName=dyn_hair_front_l_01_capsule, center=<3 0 0>, rotation=<rot 90 0 0>}]
 dyn_hair_front_l_02 [{shapeType=Capsule, length=1.375, radius=2.625, shapeName=dyn_hair_front_l_02_capsule, center=<4 0 0>, rotation=<rot 90 0 0>}]
 dyn_hair_front_l_MSTR [{shapeType=Capsule, length=0.505, radius=0.505, shapeName=dyn_hair_front_l_MSTR_capsule, rotation=<rot 0 -90 0>}]
 dyn_hair_front_r_01 [{shapeType=Capsule, length=3.125, radius=2.875, shapeName=dyn_hair_front_r_01_capsule, center=<3 0 0>, rotation=<rot 90 0 0>}]
 dyn_hair_front_r_02 [{shapeType=Capsule, length=1.375, radius=2.625, shapeName=dyn_hair_front_r_02_capsule, center=<4 0 0>, rotation=<rot 90 0 0>}]
 dyn_hair_front_r_MSTR [{shapeType=Capsule, length=0.505, radius=0.505, shapeName=dyn_hair_front_r_MSTR_capsule, rotation=<rot 0 -90 0>}]
 dyn_headphones [
  {
   shapeType=Capsule
   length=3.375
   radius=3
   shapeName=dyn_headphones_capsule
   center=<9.5897 -2.5826 -2.9242>
   rotation=<rot 2.6882 -114.0467 -101.5138>
  }
  {
   shapeType=Capsule
   length=3.5
   radius=2.875
   shapeName=dyn_headphones_capsule
   center=<10.0449 -2.9083 4.0258>
   rotation=<rot -90 180 154.9997>
  }
 ]
 dyn_helmet_side_l_01 [{
   shapeType=Capsule
   length=1.25
   radius=2.63
   shapeName=dyn_helmet_side_l_01_capsule
   center=<2.398 -0.2384 -2.8255>
   rotation=<rot 0 -89.9999 90>
  }]
 dyn_helmet_side_l_02 [{
   shapeType=Capsule
   length=3.0304
   radius=2.9167
   shapeName=dyn_helmet_side_l_02_capsule
   center=<2.2615 -0.1673 -2.8034>
   rotation=<rot 0 -91.1791 -94.2526>
  }]
 dyn_helmet_side_r_01 [{
   shapeType=Capsule
   length=1.125
   radius=2.755
   shapeName=dyn_helmet_side_l_01_capsule
   center=<1.7666 -0.6306 -2.8255>
   rotation=<rot 0 -89.9999 90>
  }]
 dyn_helmet_side_r_02 [{
   shapeType=Capsule
   length=2.9054
   radius=3.0417
   shapeName=dyn_helmet_side_l_02_capsule
   center=<2.2686 0.1736 -2.8034>
   rotation=<rot 0 -91.1791 -94.2526>
  }]
 head [
  {shapeType=Capsule, length=3.0796, radius=6.7593, shapeName=head_capsule, center=<2.3025 0.855 -0.5075>, rotation=<rot 90 0 5.0001>}
  {shapeType=Capsule, length=5.375, radius=2, shapeName=head_capsule, center=<1.3844 6.7235 0>, rotation=<rot -90 90 -90.0001>}
 ]
 neck_01 [{shapeType=Capsule, length=5, radius=6.63, shapeName=neck_01_capsule, rotation=<rot 0 -90 0>}]
 spine_04 [{shapeType=Capsule, length=4.4066, radius=10.2626, shapeName=spine_04_capsule, center=<0.306 0.5327 0>, rotation=<rot 0 -92.6847 0>}]
 spine_05 [{shapeType=Capsule, length=12.9411, radius=8.1451, shapeName=spine_05_capsule, center=<0.357 -2.1578 -0.1551>}]
 upperarm_l [{shapeType=Capsule, length=13.0479, radius=4.5179, shapeName=upperarm_l_capsule, center=<6.5238 -1.25 0.0001>, rotation=<rot 0 -90 90>}]
 upperarm_r [{shapeType=Capsule, length=13.0479, radius=4.5179, shapeName=upperarm_r_capsule, center=<-6.2385 1.25 0.0001>, rotation=<rot 0 -90 90>}]
}
constraints [
 {bone1Name=dyn_bangs_l_01, bone2Name=dyn_bangs_l_MSTR, swing1LimitDegrees=35, swing2LimitDegrees=35, twistLimitDegrees=3}
 {bone1Name=dyn_bangs_l_02, bone2Name=dyn_bangs_l_01, swing1LimitDegrees=35, swing2LimitDegrees=35, twistLimitDegrees=3}
 {bone1Name=dyn_bangs_l_03, bone2Name=dyn_bangs_l_02, swing1LimitDegrees=35, swing2LimitDegrees=35, twistLimitDegrees=3}
 {bone1Name=dyn_bangs_l_04, bone2Name=dyn_bangs_l_03, swing1LimitDegrees=35, swing2LimitDegrees=35, twistLimitDegrees=3}
 {bone1Name=dyn_bangs_r_01, bone2Name=dyn_bangs_r_MSTR, swing1LimitDegrees=10, swing2LimitDegrees=10, twistLimitDegrees=3}
 {bone1Name=dyn_bangs_r_02, bone2Name=dyn_bangs_r_01, swing1LimitDegrees=35, swing2LimitDegrees=35, twistLimitDegrees=3}
 {bone1Name=dyn_bangs_r_03, bone2Name=dyn_bangs_r_02, swing1LimitDegrees=35, swing2LimitDegrees=35, twistLimitDegrees=3}
 {bone1Name=dyn_bangs_r_04, bone2Name=dyn_bangs_r_03, swing1LimitDegrees=35, swing2LimitDegrees=35, twistLimitDegrees=3}
 {
  bone1Name=dyn_bangs_l_MSTR
  bone2Name=head
  swing1LimitDegrees=45
  swing1Motion=Free
  swing2LimitDegrees=45
  swing2Motion=Free
  twistLimitDegrees=45
  twistMotion=Free
 }
 {
  bone1Name=dyn_bangs_r_MSTR
  bone2Name=head
  swing1LimitDegrees=45
  swing1Motion=Free
  swing2LimitDegrees=45
  swing2Motion=Free
  twistLimitDegrees=45
  twistMotion=Free
 }
 {
  bone1Name=spine_05
  bone2Name=spine_04
  swing1LimitDegrees=45
  swing1Motion=Free
  swing2LimitDegrees=45
  swing2Motion=Free
  twistLimitDegrees=45
  twistMotion=Free
 }
 {
  bone1Name=clavicle_l
  bone2Name=spine_05
  swing1LimitDegrees=45
  swing1Motion=Free
  swing2LimitDegrees=45
  swing2Motion=Free
  twistLimitDegrees=45
  twistMotion=Free
 }
 {
  bone1Name=upperarm_l
  bone2Name=clavicle_l
  swing1LimitDegrees=45
  swing1Motion=Free
  swing2LimitDegrees=45
  swing2Motion=Free
  twistLimitDegrees=45
  twistMotion=Free
 }
 {
  bone1Name=upperarm_r
  bone2Name=clavicle_r
  swing1LimitDegrees=45
  swing1Motion=Free
  swing2LimitDegrees=45
  swing2Motion=Free
  twistLimitDegrees=45
  twistMotion=Free
 }
 {
  bone1Name=clavicle_r
  bone2Name=spine_05
  swing1LimitDegrees=45
  swing1Motion=Free
  swing2LimitDegrees=45
  swing2Motion=Free
  twistLimitDegrees=45
  twistMotion=Free
 }
 {
  bone1Name=dyn_helmet_side_l_01
  bone2Name=head
  swing1LimitDegrees=45
  swing1Motion=Free
  swing2LimitDegrees=45
  swing2Motion=Free
  twistLimitDegrees=45
  twistMotion=Free
 }
 {
  bone1Name=dyn_helmet_side_r_02
  bone2Name=dyn_helmet_side_r_01
  swing1LimitDegrees=45
  swing1Motion=Free
  swing2LimitDegrees=45
  swing2Motion=Free
  twistLimitDegrees=45
  twistMotion=Free
 }
 {
  bone1Name=dyn_helmet_side_l_02
  bone2Name=dyn_helmet_side_l_01
  swing1LimitDegrees=45
  swing1Motion=Free
  swing2LimitDegrees=45
  swing2Motion=Free
  twistLimitDegrees=45
  twistMotion=Free
 }
 {
  bone1Name=dyn_helmet_side_r_01
  bone2Name=head
  swing1LimitDegrees=45
  swing1Motion=Free
  swing2LimitDegrees=45
  swing2Motion=Free
  twistLimitDegrees=45
  twistMotion=Free
 }
 {
  bone1Name=head
  bone2Name=neck_01
  swing1LimitDegrees=45
  swing1Motion=Free
  swing2LimitDegrees=45
  swing2Motion=Free
  twistLimitDegrees=45
  twistMotion=Free
 }
 {
  bone1Name=neck_01
  bone2Name=spine_05
  swing1LimitDegrees=45
  swing1Motion=Free
  swing2LimitDegrees=45
  swing2Motion=Free
  twistLimitDegrees=45
  twistMotion=Free
 }
 {
  bone1Name=dyn_headphones
  bone2Name=neck_01
  swing1LimitDegrees=45
  swing1Motion=Free
  swing2LimitDegrees=45
  swing2Motion=Free
  twistLimitDegrees=45
  twistMotion=Free
 }
 {bone1Name=dyn_hair_front_r_02, bone2Name=dyn_hair_front_r_01, swing1LimitDegrees=10, swing2LimitDegrees=35, twistLimitDegrees=5}
 {
  bone1Name=dyn_hair_front_l_MSTR
  bone2Name=head
  swing1LimitDegrees=45
  swing1Motion=Free
  swing2LimitDegrees=45
  swing2Motion=Free
  twistLimitDegrees=45
  twistMotion=Free
 }
 {bone1Name=dyn_hair_front_l_01, bone2Name=dyn_hair_front_l_MSTR, swing1LimitDegrees=10, swing2LimitDegrees=35, twistLimitDegrees=5}
 {bone1Name=dyn_hair_front_l_02, bone2Name=dyn_hair_front_l_01, swing1LimitDegrees=10, swing2LimitDegrees=35, twistLimitDegrees=5}
 {
  bone1Name=dyn_hair_front_r_MSTR
  bone2Name=head
  swing1LimitDegrees=45
  swing1Motion=Free
  swing2LimitDegrees=45
  swing2Motion=Free
  twistLimitDegrees=45
  twistMotion=Free
 }
 {bone1Name=dyn_hair_front_r_01, bone2Name=dyn_hair_front_r_MSTR, swing1LimitDegrees=10, swing2LimitDegrees=35, twistLimitDegrees=5}
]
