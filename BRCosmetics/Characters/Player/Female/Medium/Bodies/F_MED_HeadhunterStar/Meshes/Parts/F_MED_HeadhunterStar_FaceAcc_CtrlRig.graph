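# Control Rig for the F_MED_HeadhunterStar face accessory that applies pose-driven corrections to dynamic back hair and headphone bones as the head moves.
# Spherical pose readers on the head output weights for Modify Transforms nodes, offsetting and scaling hair, and rotating the headphones relative to spine_05.
# A second sequence step branches on the `rban` bool: true adds head-driven back-hair offsets, false applies a separate full-weight hair transform.
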ControlRigBlueprint /BRCosmetics/Characters/Player/Female/Medium/Bodies/F_MED_HeadhunterStar/Meshes/Parts/F_MED_HeadhunterStar_FaceAcc_CtrlRig
graphs [
 RigVMFunctionLibrary {links [], nodes []}
 RigVMModel {nodes [
   RigUnit_BeginExecution = &RigUnit_BeginExecution()
   SequenceAggregate = &RigVMFunction_Sequence(ExecuteContext=@RigUnit_BeginExecution.ExecutePin)
   n1 = &RigUnit_SphericalPoseReader(
    ActiveRegionScaleFactors=(PositiveWidth=1,NegativeWidth=0.5,PositiveHeight=1.5,NegativeHeight=1.5)
    ActiveRegionSize=0.1
    Debug=(bDrawDebug=false,DebugScale=25,DebugSegments=20,DebugThickness=0.25)
    DriverAxis=(X=0,Y=1,Z=0)
    DriverItem=(Type=Bone,Name="head")
    FalloffRegionScaleFactors=(PositiveWidth=1,NegativeWidth=0.5,PositiveHeight=1.5,NegativeHeight=1.5)
    FalloffSize=0.3
    FlipWidthScaling=false
    OptionalParentItem=(Type=Bone)
    RotationOffset=(X=90,Y=0,Z=-30)
    ExecutePin=@SequenceAggregate.A
   )
   > &RigUnit_ModifyTransforms(
    ItemToModify="((Item=(Type=Bone,Name=\"dyn_Hair_Back_03\"),Transform=(Rotation=(X=0,Y=0,Z=0,W=1),Translation=(X=3,Y=-1,Z=0),Scale3D=(X=1,Y=1.2,Z=1))),(Item=(Type=Bone,Name=\"dyn_Hair_BackL_03\"),Transform=(Rotation=(X=0,Y=0,Z=0,W=1),Translation=(X=3,Y=-1,Z=0),Scale3D=(X=1,Y ...(truncated)"
    Weight=@n1.OutputParam
   )
   > n2 = &RigUnit_SphericalPoseReader(
    ActiveRegionScaleFactors=(PositiveWidth=2.5,NegativeWidth=3,PositiveHeight=1,NegativeHeight=3)
    ActiveRegionSize=0.1
    Debug=(bDrawDebug=false,DebugScale=25,DebugSegments=20,DebugThickness=0.25)
    DriverAxis=(X=1,Y=0,Z=0)
    DriverItem=(Type=Bone,Name="head")
    FalloffRegionScaleFactors=(PositiveWidth=1,NegativeWidth=0.5,PositiveHeight=0.5,NegativeHeight=0.5)
    FalloffSize=0.45
    FlipHeightScaling=true
    OptionalParentItem=(Type=Bone,Name="spine_05")
    RotationOffset=(X=-25,Y=0,Z=0)
   )
   > &RigUnit_ModifyTransforms(
    ItemToModify=((Item=(Type=Bone,Name="dyn_headphones"),Transform=(Rotation=(X=0,Y=0,Z=0,W=1),Translation=(X=0,Y=0,Z=0),Scale3D=(X=1,Y=1,Z=1))))
    ItemToModify.0.Transform.Rotation=&RigVMFunction_MathQuaternionFromEuler(Euler=(X=0,Y=20,Z=0)).Result
    Weight=@n2.OutputParam
   )
   Branch_RigVMUnitNode = &RigVMFunction_ControlFlowBranch(ExecuteContext=@SequenceAggregate.B, Condition=RigVMVariableNode() {variable=rban}.Value)
   &RigUnit_ModifyTransforms(
    ItemToModify="((Item=(Type=Bone,Name=\"dyn_Hair_Back_02\"),Transform=(Rotation=(X=0,Y=0,Z=0,W=1),Translation=(X=0,Y=0,Z=0),Scale3D=(X=1,Y=1,Z=1))),(Item=(Type=Bone,Name=\"dyn_Hair_BackR_02\"),Transform=(Rotation=(X=0,Y=0,Z=0,W=1),Translation=(X=0,Y=0,Z=0),Scale3D=(X=1,Y=1,Z ...(truncated)"
    Weight=1
    ExecutePin=@Branch_RigVMUnitNode.False
   )
   n3 = &RigUnit_SphericalPoseReader(
    ActiveRegionScaleFactors=(PositiveWidth=1,NegativeWidth=0.5,PositiveHeight=1.5,NegativeHeight=1.5)
    ActiveRegionSize=0.1
    Debug=(bDrawDebug=false,DebugScale=25,DebugSegments=20,DebugThickness=0.25)
    DriverAxis=(X=0,Y=1,Z=0)
    DriverItem=(Type=Bone,Name="head")
    FalloffRegionScaleFactors=(PositiveWidth=1,NegativeWidth=0.5,PositiveHeight=1.5,NegativeHeight=1.5)
    FalloffSize=0.3
    FlipWidthScaling=false
    OptionalParentItem=(Type=Bone)
    RotationOffset=(X=90,Y=0,Z=-30)
    ExecutePin=@Branch_RigVMUnitNode.True
   )
   > &RigUnit_ModifyTransforms(
    ItemToModify="((Item=(Type=Bone,Name=\"dyn_Hair_Back_02\"),Transform=(Rotation=(X=0,Y=0,Z=0,W=1),Translation=(X=0,Y=-2,Z=0),Scale3D=(X=1,Y=1,Z=1))),(Item=(Type=Bone,Name=\"dyn_Hair_BackR_02\"),Transform=(Rotation=(X=0,Y=0,Z=0,W=1),Translation=(X=0,Y=-3,Z=0),Scale3D=(X=1,Y=1 ...(truncated)"
    Weight=@n3.OutputParam
   )
   &RigVMFunction_MathQuaternionFromEuler(Euler=(X=0,Y=0,Z=0))
  ]}
]
hierarchy [
 root {loc=<0 0 0>, rot=<0 0 0 1>}
 pelvis {parent=root, loc=<0 -0.5301 98.0571>, rot=<0.0127 0.707 -0.0127 -0.707>}
 spine_01 {parent=pelvis, loc=<6.1208 -0.9328 0>, rot=<0 0 0.034 -0.9994>}
 spine_02 {parent=spine_01, loc=<6.5996 -0.2095 0>, rot=<0 0 0.0238 -0.9997>}
 spine_03 {parent=spine_02, loc=<6.4914 -0.1237 0>, rot=<0 0 0.0149 -0.9999>}
 spine_04 {parent=spine_03, loc=<6.9598 0.0794 0>, rot=<0 0 -0.0132 -0.9999>}
 spine_05 {parent=spine_04, loc=<6.9987 0 0>, rot=<0 0 -0.0015 -1>}
 attach_cape {parent=spine_05, loc=<9.3216 -8.4501 0>, rot=<0.0616 -0.7044 -0.0616 -0.7044>}
 neck_01 {parent=spine_05, loc=<11.3138 0.1529 0>, rot=<0 0 -0.2003 -0.9797>}
 neck_02 {parent=neck_01, loc=<4.5627 0 0>, rot=<0 0 0.0504 -0.9987>}
 head {parent=neck_02, loc=<5.3914 0 0>, rot=<0 0 0.115 -0.9934>}
 dyn_hair_bun_r_03 {parent=head, loc=<10.919 -0.5694 8.5561>, rot=<0.0029 -0.7071 0.7071 -0.0029>}
 dyn_hair_bun_r_01 {parent=head, loc=<11.2396 -0.6017 8.0847>, rot=<0.0029 -0.7071 0.7071 -0.0029>}
 dyn_hair_bun_r_02 {parent=head, loc=<10.8137 -0.5273 8.5509>, rot=<0.0029 -0.7071 0.7071 -0.0029>}
 dyn_hair_bun_l_03 {parent=head, loc=<11.3207 -0.5513 -8.4877>, rot=<-0.0029 0.7071 0.7071 -0.0029>}
 dyn_hair_bun_l_01 {parent=head, loc=<11.6931 -0.7995 -8.0522>, rot=<-0.0029 0.7071 0.7071 -0.0029>}
 dyn_hair_bun_l_02 {parent=head, loc=<11.3147 -0.4952 -8.3369>, rot=<-0.0029 0.7071 0.7071 -0.0029>}
 dyn_helmet_strap {parent=head, loc=<6.0247 -7.296 3.4765>, rot=<0.1151 0.7165 -0.1147 -0.6784>}
 dyn_helmet_side_l_MSTR {parent=head, loc=<2.2595 1.2532 -7.4859>, rot=<0.0041 -1 0 0>}
 dyn_helmet_side_l_01 {parent=dyn_helmet_side_l_MSTR, loc=<0 0 0>, rot=<0 0 0 1>}
 dyn_helmet_side_l_02 {parent=dyn_helmet_side_l_01, loc=<4.2064 0 0>, rot=<0 0 0 1>}
 dyn_helmet_side_r_MSTR {parent=head, loc=<2.2595 1.2532 7.486>, rot=<0 0 1 -0.0041>}
 dyn_helmet_side_r_01 {parent=dyn_helmet_side_r_MSTR, loc=<0 0 0>, rot=<0 0 0 1>}
 dyn_helmet_side_r_02 {parent=dyn_helmet_side_r_01, loc=<4.2064 0 0>, rot=<0 0 0 1>}
 dyn_bangs_l_MSTR {parent=head, loc=<6.8565 8.1019 -4.3824>, rot=<-0.0576 -0.998 -0.0018 -0.0277>}
 dyn_bangs_l_01 {parent=dyn_bangs_l_MSTR, loc=<0 0 0>, rot=<0 0 0 1>}
 dyn_bangs_l_02 {parent=dyn_bangs_l_01, loc=<5.0278 0 0>, rot=<-0.0018 -0.0186 0.0258 -0.9995>}
 dyn_bangs_l_03 {parent=dyn_bangs_l_02, loc=<5.5902 0 0>, rot=<-0.0003 -0.0033 -0.009 -1>}
 dyn_bangs_l_04 {parent=dyn_bangs_l_03, loc=<5.5914 0 0>, rot=<-0.0022 -0.0283 0.0131 -0.9995>}
 dyn_bangs_r_MSTR {parent=head, loc=<6.6129 6.2663 5.6871>, rot=<-0.0319 -0.9992 0.0009 0.0234>}
 dyn_bangs_r_01 {parent=dyn_bangs_r_MSTR, loc=<-0.008 -0.0779 0.0511>, rot=<0 0 0 1>}
 dyn_bangs_r_02 {parent=dyn_bangs_r_01, loc=<5.2078 0 0>, rot=<0.0014 0.0198 0.0017 -0.9998>}
 dyn_bangs_r_03 {parent=dyn_bangs_r_02, loc=<5.5533 0 0>, rot=<0 -0.0002 -0.0073 -1>}
 dyn_bangs_r_04 {parent=dyn_bangs_r_03, loc=<5.5485 0 0>, rot=<0.0019 0.0254 0.0095 -0.9996>}
 dyn_Hair_BackL_MSTR {parent=head, loc=<3.9668 -4.3121 -4.0889>, rot=<0.2933 -0.9531 0.0215 -0.0721>}
 dyn_Hair_BackL_01 {parent=dyn_Hair_BackL_MSTR, loc=<0 0 0>, rot=<0 0 0 1>}
 dyn_Hair_BackL_02 {parent=dyn_Hair_BackL_01, loc=<2.5158 0 0>, rot=<0.0055 -0.0135 -0.2024 -0.9792>}
 dyn_Hair_BackL_03 {parent=dyn_Hair_BackL_02, loc=<3.7818 0 0>, rot=<-0.006 0.0243 0.0596 -0.9979>}
 dyn_Hair_BackL_04 {parent=dyn_Hair_BackL_03, loc=<6.853 0 0>, rot=<0.0097 -0.0262 0.0556 -0.9981>}
 dyn_Hair_BackL_05 {parent=dyn_Hair_BackL_04, loc=<6.6725 0 0>, rot=<0 0 0 1>}
 dyn_Hair_Back_MSTR {parent=head, loc=<4.4961 -4.542 0.5844>, rot=<0.2966 -0.955 0 0>}
 dyn_Hair_Back_01 {parent=dyn_Hair_Back_MSTR, loc=<0 0 0>, rot=<0 0 0 1>}
 dyn_Hair_Back_02 {parent=dyn_Hair_Back_01, loc=<2.4962 0 0>, rot=<0 0 -0.1793 -0.9838>}
 dyn_Hair_Back_03 {parent=dyn_Hair_Back_02, loc=<4.385 0 0>, rot=<-0.0032 0.0118 0.0322 -0.9994>}
 dyn_Hair_Back_04 {parent=dyn_Hair_Back_03, loc=<6.8932 0 0>, rot=<-0.0029 0.0092 0.0068 -0.9999>}
 dyn_Hair_Back_05 {parent=dyn_Hair_Back_04, loc=<7.1689 0 0>, rot=<0 0 0 1>}
 dyn_Hair_BackR_MSTR {parent=head, loc=<4.18 -3.7157 4.9819>, rot=<0.2926 -0.9549 -0.0147 0.0493>}
 dyn_Hair_BackR_01 {parent=dyn_Hair_BackR_MSTR, loc=<0 0 0>, rot=<0 0 0 1>}
 dyn_Hair_BackR_02 {parent=dyn_Hair_BackR_01, loc=<2.5457 0 0>, rot=<-0.0005 0.0012 -0.2042 -0.9789>}
 dyn_Hair_BackR_03 {parent=dyn_Hair_BackR_02, loc=<4.0296 0 0>, rot=<0.0072 -0.0343 0.0315 -0.9989>}
 dyn_Hair_BackR_04 {parent=dyn_Hair_BackR_03, loc=<6.6174 0 0>, rot=<-0.0022 0.0072 0.062 -0.998>}
 dyn_Hair_BackR_05 {parent=dyn_Hair_BackR_04, loc=<4.6083 0 0>, rot=<0 0 0 1>}
 dyn_hair_front_l_MSTR {parent=head, loc=<-6.9179 -2.3629 -7.0352>, rot=<-0.2319 -0.664 -0.655 -0.2764>}
 dyn_hair_front_l_01 {parent=dyn_hair_front_l_MSTR, loc=<0 0 0>, rot=<0 0 0 1>}
 dyn_hair_front_l_02 {parent=dyn_hair_front_l_01, loc=<6.1915 0 0>, rot=<0 0 0 1>}
 dyn_hair_front_r_MSTR {parent=head, loc=<-7.2436 -1.9206 6.6793>, rot=<0.3024 0.6466 0.6697 0.2049>}
 dyn_hair_front_r_01 {parent=dyn_hair_front_r_MSTR, loc=<0 0 0>, rot=<0 0 0 1>}
 dyn_hair_front_r_02 {parent=dyn_hair_front_r_01, loc=<6.2778 0 0>, rot=<0 0 0 1>}
 dyn_headphones_MSTR {parent=neck_01, loc=<2.997 -4.3187 0>, rot=<0 0 -0.8408 -0.5414>}
 dyn_headphones {parent=dyn_headphones_MSTR, loc=<0 0 0>, rot=<0 0 0 1>}
 attach_backpack {parent=spine_05, loc=<0 -8.35 0>, rot=<-0.0093 -0.707 0.0093 -0.707>}
 clavicle_l {parent=spine_05, loc=<5.4831 -0.1053 -3.2946>, rot=<0.0465 -0.7289 -0.009 -0.6829>}
 upperarm_l {parent=clavicle_l, loc=<9.8658 -1.2136 -0.4105>, rot=<0.0164 -0.3245 -0.0269 -0.9454>}
 upperarm_twist_01_l {parent=upperarm_l, loc=<7.2135 -0.1134 0>, rot=<0 0 0 1>}
 clavicle_r {parent=spine_05, loc=<5.4831 -0.1053 3.2946>, rot=<0.7289 0.0465 -0.6829 0.009>}
 upperarm_r {parent=clavicle_r, loc=<-9.8658 1.2136 0.4105>, rot=<0.0164 -0.3245 -0.0269 -0.9454>}
 upperarm_twist_01_r {parent=upperarm_r, loc=<-7.2135 0.1134 0>, rot=<0 0 0 -1>}
 dyn_headphones_TGT {parent=spine_05, loc=<7.3759 5.0306 0>, rot=<0.0284 -0.7065 -0.0284 -0.7065>}
]
variables [rban {type=bool, default=False}]
