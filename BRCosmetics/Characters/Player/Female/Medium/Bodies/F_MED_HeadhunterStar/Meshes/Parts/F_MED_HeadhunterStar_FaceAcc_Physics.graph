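# Physics asset for the face accessory mesh of the female medium HeadhunterStar character, covering three back-hair bone chains (left, center, right).
# Each chain runs from a master bone on spine_05 through five links, with free rotation at the small upper links and tighter swing and twist limits on the three larger lower ones.
# Spine, clavicle and upper-arm capsules supply torso and shoulder shapes.
# Two right-chain links carry an extra angled capsule.
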PhysicsAsset /BRCosmetics/Characters/Player/Female/Medium/Bodies/F_MED_HeadhunterStar/Meshes/Parts/F_MED_HeadhunterStar_FaceAcc_Physics
bodies {
 clavicle_l [{
   shapeType=Capsule
   length=5.0831
   radius=4.9647
   shapeName=clavicle_l_capsule
   center=<6.5243 -0.6953 0.104>
   rotation=<rot 16.9211 87.287 -90>
  }]
 clavicle_r [{
   shapeType=Capsule
   length=5.9909
   radius=5.2737
   shapeName=clavicle_r_capsule
   center=<-6.4341 0.3237 0.8599>
   rotation=<rot 3.2384 98.0696 -90>
  }]
 dyn_Hair_BackL_01 [{shapeType=Capsule, length=0.505, radius=0.505, shapeName=dyn_Hair_BackL_01_capsule}]
 dyn_Hair_BackL_02 [{shapeType=Capsule, length=0.505, radius=0.505, shapeName=dyn_Hair_BackL_02_capsule}]
 dyn_Hair_BackL_03 [{shapeType=Capsule, length=2.625, radius=3, shapeName=dyn_Hair_BackL_03_capsule, center=<5 -2 0>, rotation=<rot 90 0 0>}]
 dyn_Hair_BackL_04 [{shapeType=Capsule, length=2.625, radius=3, shapeName=dyn_Hair_BackL_04_capsule, center=<5 -2 0>, rotation=<rot 90 0 0>}]
 dyn_Hair_BackL_05 [{shapeType=Capsule, length=2.625, radius=3, shapeName=dyn_Hair_BackL_05_capsule, center=<5 -2 0>, rotation=<rot 90 0 0>}]
 dyn_Hair_BackL_MSTR [{shapeType=Capsule, length=0.505, radius=0.505, shapeName=dyn_Hair_BackL_MSTR_capsule}]
 dyn_Hair_BackR_01 [{shapeType=Capsule, length=0.505, radius=0.505, shapeName=dyn_Hair_BackR_01_capsule}]
 dyn_Hair_BackR_02 [{shapeType=Capsule, length=0.505, radius=0.505, shapeName=dyn_Hair_BackR_02_capsule}]
 dyn_Hair_BackR_03 [
  {shapeType=Capsule, length=2.625, radius=3, shapeName=dyn_Hair_BackR_03_capsule, center=<5 -2 0>, rotation=<rot 90 0 0>}
  {
   shapeType=Capsule
   length=1.625
   radius=2.375
   shapeName=dyn_Hair_BackR_03_capsule
   center=<6.9415 0.0633 -1.933>
   rotation=<rot 70 -19.9998 0>
  }
 ]
 dyn_Hair_BackR_04 [
  {shapeType=Capsule, length=2.375, radius=3.25, shapeName=dyn_Hair_BackR_04_capsule, center=<5 -2 0>, rotation=<rot 90 0 0>}
  {
   shapeType=Capsule
   length=2.75
   radius=2.25
   shapeName=dyn_Hair_BackR_04_capsule
   center=<6.3934 -1.029 -4.3925>
   rotation=<rot 90 110.7484 115.7482>
  }
 ]
 dyn_Hair_BackR_05 [{shapeType=Capsule, length=2.375, radius=3.25, shapeName=dyn_Hair_BackR_05_capsule, center=<5 -2 0>, rotation=<rot 90 0 0>}]
 dyn_Hair_BackR_MSTR [{shapeType=Capsule, length=0.505, radius=0.505, shapeName=dyn_Hair_BackR_MSTR_capsule}]
 dyn_Hair_Back_01 [{shapeType=Capsule, length=0.505, radius=0.505, shapeName=dyn_Hair_Back_01_capsule}]
 dyn_Hair_Back_02 [{shapeType=Capsule, length=0.505, radius=0.505, shapeName=dyn_Hair_Back_02_capsule}]
 dyn_Hair_Back_03 [{shapeType=Capsule, length=2.625, radius=3, shapeName=dyn_Hair_Back_03_capsule, center=<5 -2 0>, rotation=<rot 90 0 0>}]
 dyn_Hair_Back_04 [{shapeType=Capsule, length=2.625, radius=3, shapeName=dyn_Hair_Back_04_capsule, center=<5 -2 0>, rotation=<rot 90 0 0>}]
 dyn_Hair_Back_05 [{shapeType=Capsule, length=2.625, radius=3, shapeName=dyn_Hair_Back_05_capsule, center=<5 -2 0>, rotation=<rot 90 0 0>}]
 dyn_Hair_Back_MSTR [{shapeType=Capsule, length=0.505, radius=0.505, shapeName=dyn_Hair_Back_MSTR_capsule}]
 spine_04 [{shapeType=Capsule, length=4.7816, radius=9.8876, shapeName=spine_04_capsule, center=<0.3234 0.9034 0>, rotation=<rot 0 -92.6847 0>}]
 spine_05 [{shapeType=Capsule, length=13.0661, radius=8.0201, shapeName=spine_05_capsule, center=<0.357 -0.7071 -0.1551>}]
 upperarm_l [{shapeType=Capsule, length=13.4229, radius=4.1429, shapeName=upperarm_l_capsule, center=<5 -1.25 0.0001>, rotation=<rot 0 -90 90>}]
 upperarm_r [{shapeType=Capsule, length=10.4229, radius=4.6429, shapeName=upperarm_r_capsule, center=<-5.5837 1.25 0.0001>, rotation=<rot 0 -90 90>}]
}
constraints [
 {
  bone1Name=dyn_Hair_BackL_01
  bone2Name=dyn_Hair_BackL_MSTR
  swing1LimitDegrees=45
  swing1Motion=Free
  swing2LimitDegrees=45
  swing2Motion=Free
  twistLimitDegrees=45
  twistMotion=Free
 }
 {
  bone1Name=dyn_Hair_BackL_02
  bone2Name=dyn_Hair_BackL_01
  swing1LimitDegrees=45
  swing1Motion=Free
  swing2LimitDegrees=45
  swing2Motion=Free
  twistLimitDegrees=45
  twistMotion=Free
 }
 {bone1Name=dyn_Hair_BackL_03, bone2Name=dyn_Hair_BackL_02, swing1LimitDegrees=35, swing2LimitDegrees=15, twistLimitDegrees=5}
 {bone1Name=dyn_Hair_BackL_04, bone2Name=dyn_Hair_BackL_03, swing1LimitDegrees=35, swing2LimitDegrees=15, twistLimitDegrees=5}
 {bone1Name=dyn_Hair_BackL_05, bone2Name=dyn_Hair_BackL_04, swing1LimitDegrees=35, swing2LimitDegrees=15, twistLimitDegrees=5}
 {
  bone1Name=dyn_Hair_BackR_01
  bone2Name=dyn_Hair_BackR_MSTR
  swing1LimitDegrees=45
  swing1Motion=Free
  swing2LimitDegrees=45
  swing2Motion=Free
  twistLimitDegrees=45
  twistMotion=Free
 }
 {
  bone1Name=dyn_Hair_BackR_02
  bone2Name=dyn_Hair_BackR_01
  swing1LimitDegrees=45
  swing1Motion=Free
  swing2LimitDegrees=45
  swing2Motion=Free
  twistLimitDegrees=45
  twistMotion=Free
 }
 {bone1Name=dyn_Hair_BackR_03, bone2Name=dyn_Hair_BackR_02, swing1LimitDegrees=35, swing2LimitDegrees=15, twistLimitDegrees=5}
 {bone1Name=dyn_Hair_BackR_04, bone2Name=dyn_Hair_BackR_03, swing1LimitDegrees=35, swing2LimitDegrees=15, twistLimitDegrees=5}
 {bone1Name=dyn_Hair_BackR_05, bone2Name=dyn_Hair_BackR_04, swing1LimitDegrees=35, swing2LimitDegrees=15, twistLimitDegrees=5}
 {
  bone1Name=dyn_Hair_Back_01
  bone2Name=dyn_Hair_Back_MSTR
  swing1LimitDegrees=45
  swing1Motion=Free
  swing2LimitDegrees=45
  swing2Motion=Free
  twistLimitDegrees=45
  twistMotion=Free
 }
 {
  bone1Name=dyn_Hair_Back_02
  bone2Name=dyn_Hair_Back_01
  swing1LimitDegrees=45
  swing1Motion=Free
  swing2LimitDegrees=45
  swing2Motion=Free
  twistLimitDegrees=45
  twistMotion=Free
 }
 {bone1Name=dyn_Hair_Back_03, bone2Name=dyn_Hair_Back_02, swing1LimitDegrees=35, swing2LimitDegrees=15, twistLimitDegrees=5}
 {bone1Name=dyn_Hair_Back_04, bone2Name=dyn_Hair_Back_03, swing1LimitDegrees=35, swing2LimitDegrees=15, twistLimitDegrees=5}
 {bone1Name=dyn_Hair_Back_05, bone2Name=dyn_Hair_Back_04, swing1LimitDegrees=35, swing2LimitDegrees=15, twistLimitDegrees=5}
 {
  bone1Name=dyn_Hair_BackL_MSTR
  bone2Name=spine_05
  swing1LimitDegrees=45
  swing1Motion=Free
  swing2LimitDegrees=45
  swing2Motion=Free
  twistLimitDegrees=45
  twistMotion=Free
 }
 {
  bone1Name=dyn_Hair_Back_MSTR
  bone2Name=spine_05
  swing1LimitDegrees=45
  swing1Motion=Free
  swing2LimitDegrees=45
  swing2Motion=Free
  twistLimitDegrees=45
  twistMotion=Free
 }
 {
  bone1Name=dyn_Hair_BackR_MSTR
  bone2Name=spine_05
  swing1LimitDegrees=45
  swing1Motion=Free
  swing2LimitDegrees=45
  swing2Motion=Free
  twistLimitDegrees=45
  twistMotion=Free
 }
 {
  bone1Name=spine_05
  bone2Name=spine_04
  swing1LimitDegrees=45
  swing1Motion=Free
  swing2LimitDegrees=45
  swing2Motion=Free
  twistLimitDegrees=45
  twistMotion=Free
 }
 {
  bone1Name=clavicle_r
  bone2Name=spine_05
  swing1LimitDegrees=45
  swing1Motion=Free
  swing2LimitDegrees=45
  swing2Motion=Free
  twistLimitDegrees=45
  twistMotion=Free
 }
 {
  bone1Name=upperarm_r
  bone2Name=clavicle_r
  swing1LimitDegrees=45
  swing1Motion=Free
  swing2LimitDegrees=45
  swing2Motion=Free
  twistLimitDegrees=45
  twistMotion=Free
 }
 {
  bone1Name=clavicle_l
  bone2Name=spine_05
  swing1LimitDegrees=45
  swing1Motion=Free
  swing2LimitDegrees=45
  swing2Motion=Free
  twistLimitDegrees=45
  twistMotion=Free
 }
 {
  bone1Name=upperarm_l
  bone2Name=clavicle_l
  swing1LimitDegrees=45
  swing1Motion=Free
  swing2LimitDegrees=45
  swing2Motion=Free
  twistLimitDegrees=45
  twistMotion=Free
 }
]
