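# Post-copy secondary-motion AnimBP for the HeadhunterStar face accessory's hair, headphones and helmet parts.
# It copies the attached parent's pose, then chains bone constraints, AnimDynamics, a Control Rig and two RigidBody simulations, reapplying back-hair constraints last.
# Jumping or falling tilts the helmet sides, and a front-end flag nudges the right bangs.
# Each update sets gravity through a deprecated function and zeroes the second RigidBody's motion scales when backbling is detected.
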AnimBlueprint /BRCosmetics/Characters/Player/Female/Medium/Bodies/F_MED_HeadhunterStar/Meshes/Parts/F_MED_HeadhunterStar_FaceAcc_AnimBP
target_skeleton=&/F_MED_HeadhunterStar_FaceAcc_Skeleton
graphs [
 AnimGraph {nodes [
   n1 = "Constraint - dyn_Hair_Back_01 (PP)"!(
    ComponentPose=Local To Component(LocalPose=Copy Pose From Mesh() {bUseAttachedParent=true})
    Alpha=1
    ConstraintWeights_0=0.6
    ConstraintWeights_1:float=0.4
   ) -> (Pose) {
    alphaBoolBlend {blendOption=Linear}
    boneToModify {boneName=dyn_Hair_Back_01}
    constraintSetup [{targetBone {boneName=head}}, {targetBone {boneName=spine_05}}]
    constraintWeights [1,1]
   }
   n2 = "Constraint - dyn_Hair_Back_02 (PP)"!(ComponentPose=@n1, Alpha=1, ConstraintWeights_0=0.4, ConstraintWeights_1:float=0.6) -> (Pose) {
    alphaBoolBlend {blendOption=Linear}
    boneToModify {boneName=dyn_Hair_Back_02}
    constraintSetup [{targetBone {boneName=head}}, {targetBone {boneName=spine_05}}]
    constraintWeights [1,1]
   }
   n3 = "Constraint - dyn_Hair_Back_03 (PP)"!(ComponentPose=@n2, Alpha=1, ConstraintWeights_0=0.6, ConstraintWeights_1:float=0.4) -> (Pose) {
    alphaBoolBlend {blendOption=Linear}
    boneToModify {boneName=dyn_Hair_Back_03}
    constraintSetup [{targetBone {boneName=spine_05}}, {targetBone {boneName=spine_05}}]
    constraintWeights [1,1]
   }
   n4 = "Constraint - dyn_Hair_Back_04 (PP)"!(ComponentPose=@n3, Alpha=1, ConstraintWeights_0=1, ConstraintWeights_1:float=0) -> (Pose) {
    alphaBoolBlend {blendOption=Linear}
    boneToModify {boneName=dyn_Hair_Back_04}
    constraintSetup [{targetBone {boneName=spine_05}}, {targetBone {boneName=spine_05}}]
    constraintWeights [1,1]
   }
   n5 = "Constraint - dyn_Hair_BackL_01 (PP)"!(ComponentPose=@n4, Alpha=1, ConstraintWeights_0=0.3, ConstraintWeights_1:float=0.7) -> (Pose) {
    alphaBoolBlend {blendOption=Linear}
    boneToModify {boneName=dyn_Hair_BackL_01}
    constraintSetup [{targetBone {boneName=head}}, {targetBone {boneName=dyn_Hair_Back_01}}]
    constraintWeights [1,1]
   }
   n6 = "Constraint - dyn_Hair_BackL_02 (PP)"!(ComponentPose=@n5, Alpha=1, ConstraintWeights_0=0.3, ConstraintWeights_1:float=0.7) -> (Pose) {
    alphaBoolBlend {blendOption=Linear}
    boneToModify {boneName=dyn_Hair_BackL_02}
    constraintSetup [{targetBone {boneName=head}}, {targetBone {boneName=dyn_Hair_Back_02}}]
    constraintWeights [1,1]
   }
   n7 = "Constraint - dyn_Hair_BackL_03 (PP)"!(ComponentPose=@n6, Alpha=1, ConstraintWeights_0=0.5, ConstraintWeights_1:float=0.5) -> (Pose) {
    alphaBoolBlend {blendOption=Linear}
    boneToModify {boneName=dyn_Hair_BackL_03}
    constraintSetup [{targetBone {boneName=spine_05}}, {targetBone {boneName=dyn_Hair_Back_03}}]
    constraintWeights [1,1]
   }
   n8 = "Constraint - dyn_Hair_BackL_04 (PP)"!(ComponentPose=@n7, Alpha=1, ConstraintWeights_0=0.8, ConstraintWeights_1:float=0.2) -> (Pose) {
    alphaBoolBlend {blendOption=Linear}
    boneToModify {boneName=dyn_Hair_BackL_04}
    constraintSetup [{targetBone {boneName=spine_05}}, {targetBone {boneName=dyn_Hair_Back_04}}]
    constraintWeights [1,1]
   }
   n9 = "Constraint - dyn_Hair_BackR_01 (PP)"!(ComponentPose=@n8, Alpha=1, ConstraintWeights_0=0.3, ConstraintWeights_1:float=0.7) -> (Pose) {
    alphaBoolBlend {blendOption=Linear}
    boneToModify {boneName=dyn_Hair_BackR_01}
    constraintSetup [{targetBone {boneName=head}}, {targetBone {boneName=dyn_Hair_Back_01}}]
    constraintWeights [1,1]
   }
   n10 = "Constraint - dyn_Hair_BackR_02 (PP)"!(ComponentPose=@n9, Alpha=1, ConstraintWeights_0=0.3, ConstraintWeights_1:float=0.7) -> (Pose) {
    alphaBoolBlend {blendOption=Linear}
    boneToModify {boneName=dyn_Hair_BackR_02}
    constraintSetup [{targetBone {boneName=head}}, {targetBone {boneName=dyn_Hair_Back_02}}]
    constraintWeights [1,1]
   }
   n11 = "Constraint - dyn_Hair_BackR_03 (PP)"!(ComponentPose=@n10, Alpha=1, ConstraintWeights_0=0.5, ConstraintWeights_1:float=0.5) -> (Pose) {
    alphaBoolBlend {blendOption=Linear}
    boneToModify {boneName=dyn_Hair_BackR_03}
    constraintSetup [{targetBone {boneName=spine_05}}, {targetBone {boneName=dyn_Hair_Back_03}}]
    constraintWeights [1,1]
   }
   n12 = "Constraint - dyn_Hair_BackR_04 (PP)"!(ComponentPose=@n11, Alpha=1, ConstraintWeights_0=0.8, ConstraintWeights_1:float=0.2) -> (Pose) {
    alphaBoolBlend {blendOption=Linear}
    boneToModify {boneName=dyn_Hair_BackR_04}
    constraintSetup [{targetBone {boneName=spine_05}}, {targetBone {boneName=dyn_Hair_Back_04}}]
    constraintWeights [1,1]
   }
   n13 = "Constraint - dyn_hair_front_r_MSTR (PP)"!(ComponentPose=@n12, Alpha=1, ConstraintWeights_0=0.7, ConstraintWeights_1:float=0.3) -> (Pose) {
    alphaBoolBlend {blendOption=Linear}
    boneToModify {boneName=dyn_hair_front_r_MSTR}
    constraintSetup [{targetBone {boneName=spine_05}}, {targetBone {boneName=head}}]
    constraintWeights [1,1]
   }
   n14 = "Constraint - dyn_hair_front_l_MSTR (PP)"!(ComponentPose=@n13, Alpha=1, ConstraintWeights_0=0.6, ConstraintWeights_1:float=0.4) -> (Pose) {
    alphaBoolBlend {blendOption=Linear}
    boneToModify {boneName=dyn_hair_front_l_MSTR}
    constraintSetup [{targetBone {boneName=spine_05}}, {targetBone {boneName=head}}]
    constraintWeights [1,1]
   }
   n15 = "Constraint - dyn_headphones_TGT (PPPPPPP)"!(
    ComponentPose=@n14
    Alpha=1
    ConstraintWeights_0=0.0356
    ConstraintWeights_1:float=0.0736
    ConstraintWeights_2:float=0.6758
    ConstraintWeights_3:float=0.0057
    ConstraintWeights_4:float=0.0056
    ConstraintWeights_5:float=0.1825
    ConstraintWeights_6:float=0.0212
   ) -> (Pose) {
    alphaBoolBlend {blendOption=Linear}
    boneToModify {boneName=dyn_headphones_TGT}
    constraintSetup [
     {targetBone {boneName=neck_02}}
     {targetBone {boneName=neck_01}}
     {targetBone {boneName=spine_05}}
     {targetBone {boneName=clavicle_l}}
     {targetBone {boneName=clavicle_r}}
     {targetBone {boneName=spine_04}}
     {targetBone {boneName=spine_03}}
    ]
    constraintWeights [1,1,1,1,1,1,1]
   }
   n16 = Look At - Bone: dyn_headphones_MSTR(ComponentPose=@n15, Alpha=1) {
    alphaBoolBlend {blendOption=Linear}
    boneToModify {boneName=dyn_headphones_MSTR}
    lookAtLocation {x=0}
    lookAtTarget {boneReference {boneName=dyn_headphones_TGT}}
    lookAt_Axis {axis=<1 0>}
   }
   n17 = Anim Dynamics - Bone: dyn_headphones(ComponentPose=@n16, Alpha=0.5, ExternalForce="0,0,0") {
    angularDampingOverride=0.98
    angularSpringConstant=50
    bAngularSpring=true
    bGravityOverrideInSimSpace=true
    bOverrideAngularDamping=true
    simulationSpace=RootRelative
    alphaBoolBlend {blendOption=Linear}
    boundBone {boneName=dyn_headphones}
    componentAppliedLinearAccClamp=<3500 3500 3500>
    componentLinearAccScale=<0.6 0.6 1.2>
    componentLinearVelScale=<0.6 0.6 1.2>
    physicsBodyDefinitions [{
      sphereCollisionRadius=0
      boundBone {boneName=dyn_headphones}
      boxExtents=<15 10 10>
      constraintSetup {angularLimitsMax=<1 5 0>, angularLimitsMin=<-1 -5 -15>, angularTarget=<1 0 0>}
      localJointOffset=<2 0 0>
     }]
    relativeSpaceBone {boneName=spine_05}
   }
   n18 = Get jumping() -> (jumping:double)
   n19 = "Transform (Modify) Bone - Bone: dyn_helmet_side_l_MSTR"(ComponentPose=@n17, Alpha=@n18, Translation="0,0,0", Rotation="55,0,0", Scale="1,1,1") {
    rotationMode=BMM_Additive
    rotationSpace=BCS_BoneSpace
    alphaBoolBlend {blendOption=Linear}
    alphaScaleBiasClamp {bInterpResult=true, interpSpeedIncreasing=2}
    boneToModify {boneName=dyn_helmet_side_l_MSTR}
    rotation {pitch=55}
   }
   n20 = "Transform (Modify) Bone - Bone: dyn_helmet_side_r_MSTR"(ComponentPose=@n19, Alpha=@n18, Translation="0,0,0", Rotation="55,0,0", Scale="1,1,1") {
    rotationMode=BMM_Additive
    rotationSpace=BCS_BoneSpace
    alphaBoolBlend {blendOption=Linear}
    alphaScaleBiasClamp {bInterpResult=true, interpSpeedIncreasing=2}
    boneToModify {boneName=dyn_helmet_side_r_MSTR}
    rotation {pitch=55}
   }
   n21 = Anim Dynamics - Bone: dyn_helmet_side_l_01(ComponentPose=@n20, Alpha=1, ExternalForce="0,0,0") {
    angularDampingOverride=0.98
    angularSpringConstant=50
    bAngularSpring=true
    bGravityOverrideInSimSpace=true
    bOverrideAngularDamping=true
    simulationSpace=RootRelative
    alphaBoolBlend {blendOption=Linear}
    boundBone {boneName=dyn_helmet_side_l_01}
    componentAppliedLinearAccClamp=<3500 3500 3500>
    componentLinearAccScale=<0.6 0.6 1.2>
    componentLinearVelScale=<0.6 0.6 1.2>
    physicsBodyDefinitions [{
      sphereCollisionRadius=0
      boundBone {boneName=dyn_helmet_side_l_01}
      boxExtents=<20 15 15>
      constraintSetup {angularLimitsMax=<1 3 5>, angularLimitsMin=<-1 -15 0>, angularTarget=<1 0 0>}
      localJointOffset=<1 0 0>
     }]
    relativeSpaceBone {boneName=root}
   }
   n22 = Anim Dynamics - Bone: dyn_helmet_side_l_02(ComponentPose=@n21, Alpha=1, ExternalForce="0,0,0") {
    angularDampingOverride=0.98
    angularSpringConstant=50
    bAngularSpring=true
    bGravityOverrideInSimSpace=true
    bOverrideAngularDamping=true
    simulationSpace=RootRelative
    alphaBoolBlend {blendOption=Linear}
    boundBone {boneName=dyn_helmet_side_l_02}
    componentAppliedLinearAccClamp=<3500 3500 3500>
    componentLinearAccScale=<0.6 0.6 1.2>
    componentLinearVelScale=<0.6 0.6 1.2>
    physicsBodyDefinitions [{
      sphereCollisionRadius=0
      boundBone {boneName=dyn_helmet_side_l_02}
      boxExtents=<20 15 15>
      constraintSetup {angularLimitsMax=<1 3 5>, angularLimitsMin=<-1 -15 0>, angularTarget=<1 0 0>}
      localJointOffset=<1 0 0>
     }]
    relativeSpaceBone {boneName=root}
   }
   n23 = Anim Dynamics - Bone: dyn_helmet_side_r_01(ComponentPose=@n22, Alpha=1, ExternalForce="0,0,0") {
    angularDampingOverride=0.98
    angularSpringConstant=50
    bAngularSpring=true
    bGravityOverrideInSimSpace=true
    bOverrideAngularDamping=true
    simulationSpace=RootRelative
    alphaBoolBlend {blendOption=Linear}
    boundBone {boneName=dyn_helmet_side_r_01}
    componentAppliedLinearAccClamp=<3500 3500 3500>
    componentLinearAccScale=<0.6 0.6 1.2>
    componentLinearVelScale=<0.6 0.6 1.2>
    physicsBodyDefinitions [{
      sphereCollisionRadius=0
      boundBone {boneName=dyn_helmet_side_r_01}
      boxExtents=<20 15 15>
      constraintSetup {angularLimitsMax=<1 3 5>, angularLimitsMin=<-1 -15 0>, angularTarget=<1 0 0>}
      localJointOffset=<1 0 0>
     }]
    relativeSpaceBone {boneName=root}
   }
   n24 = Anim Dynamics - Bone: dyn_helmet_side_r_02(ComponentPose=@n23, Alpha=1, ExternalForce="0,0,0") {
    angularDampingOverride=0.98
    angularSpringConstant=50
    bAngularSpring=true
    bGravityOverrideInSimSpace=true
    bOverrideAngularDamping=true
    simulationSpace=RootRelative
    alphaBoolBlend {blendOption=Linear}
    boundBone {boneName=dyn_helmet_side_r_02}
    componentAppliedLinearAccClamp=<3500 3500 3500>
    componentLinearAccScale=<0.6 0.6 1.2>
    componentLinearVelScale=<0.6 0.6 1.2>
    physicsBodyDefinitions [{
      sphereCollisionRadius=0
      boundBone {boneName=dyn_helmet_side_r_02}
      boxExtents=<20 15 15>
      constraintSetup {angularLimitsMax=<1 3 5>, angularLimitsMin=<-1 -15 0>, angularTarget=<1 0 0>}
      localJointOffset=<1 0 0>
     }]
    relativeSpaceBone {boneName=root}
   }
   n25 = Control Rig!(Source=Component To Local(ComponentPose=@n24), Alpha, rban:bool=Get bRigidBodyEnabled()) -> (Pose) {
    alphaBoolBlend {blendOption=Linear}
    controlRigAssetReference {blueprintRigClass=&/F_MED_HeadhunterStar_FaceAcc_CtrlRig.F_MED_HeadhunterStar_FaceAcc_CtrlRig_C}
   }
   n26 = "Transform (Modify) Bone - Bone: dyn_bangs_r_01"!(
    ComponentPose=Local To Component(LocalPose=@n25)
    bAlphaBoolEnabled:bool=Get bIsInFrontEnd()
    Translation="0,0,0"
    Rotation="0,3.5,0"
    Scale="1,1,1"
   ) -> (Pose) {
    alphaInputType=Bool
    rotationMode=BMM_Additive
    rotationSpace=BCS_BoneSpace
    alphaBoolBlend {blendInTime=0.1, blendOption=Linear, blendOutTime=0.1}
    alphaScaleBiasClamp {bInterpResult=true, interpSpeedIncreasing=2}
    boneToModify {boneName=dyn_bangs_r_01}
    rotation {yaw=3.5}
   }
   n27 = RigidBody!(ComponentPose=@n26, Alpha, OverrideWorldGravity:Vector="0,0,-980", ExternalForce) -> (Pose) {
    bClampLinearTranslationLimitToRefPose=true
    bOverrideWorldGravity=true
    overridePhysicsAsset=&/F_MED_HeadhunterStar_FR_FaceAcc_Physics
    simulationSpace=BaseBoneSpace
    alphaBoolBlend {blendOption=Linear}
    baseBoneRef {boneName=spine_05}
    componentAppliedLinearAccClamp=<1000 1000 1000>
    componentLinearAccScale=<0.2 0.2 0.3>
    componentLinearVelScale=<0.1 0.1 0.2>
   }
   n28 = RigidBody!(
    ComponentPose=@n27
    Alpha
    OverrideWorldGravity=Get GravityOverride()
    ExternalForce
    ComponentLinearAccScale:Vector=Get ACC() -> (ACC:Vector)
    ComponentLinearVelScale:Vector=Get VEL() -> (VEL:Vector)
   ) -> (Pose) {
    bClampLinearTranslationLimitToRefPose=true
    bOverrideWorldGravity=true
    overridePhysicsAsset=&/F_MED_HeadhunterStar_FaceAcc_Physics
    simulationSpace=BaseBoneSpace
    alphaBoolBlend {blendOption=Linear}
    baseBoneRef {boneName=spine_05}
    componentAppliedLinearAccClamp=<2500 2500 2500>
    componentLinearAccScale=<0.3 0.3 0.5>
    componentLinearVelScale=<0.3 0.3 0.5>
    overrideWorldGravity {z=-800}
   }
   n29 = Anim Dynamics - Bone: dyn_hair_bun_r_01(ComponentPose=@n28, Alpha=1, ExternalForce="0,0,0") {
    angularDampingOverride=0.98
    angularSpringConstant=50
    bAngularSpring=true
    bGravityOverrideInSimSpace=true
    bOverrideAngularDamping=true
    simulationSpace=RootRelative
    alphaBoolBlend {blendOption=Linear}
    boundBone {boneName=dyn_hair_bun_r_01}
    componentAppliedLinearAccClamp=<3500 3500 3500>
    componentLinearAccScale=<0.6 0.6 0.6>
    componentLinearVelScale=<0.6 0.6 0.6>
    physicsBodyDefinitions [{
      sphereCollisionRadius=0
      boundBone {boneName=dyn_hair_bun_r_01}
      boxExtents=<20 15 15>
      constraintSetup {angularLimitsMax=<1 15 5>, angularLimitsMin=<-1 -15 -15>, angularTarget=<1 0 0>}
      localJointOffset=<3 0 0>
     }]
    relativeSpaceBone {boneName=root}
   }
   n30 = Anim Dynamics - Bone: dyn_hair_bun_r_02(ComponentPose=@n29, Alpha=1, ExternalForce="0,0,0") {
    angularDampingOverride=0.98
    angularSpringConstant=50
    bAngularSpring=true
    bGravityOverrideInSimSpace=true
    bOverrideAngularDamping=true
    simulationSpace=RootRelative
    alphaBoolBlend {blendOption=Linear}
    boundBone {boneName=dyn_hair_bun_r_02}
    componentAppliedLinearAccClamp=<3500 3500 3500>
    componentLinearAccScale=<0.6 0.6 0.6>
    componentLinearVelScale=<0.6 0.6 0.6>
    physicsBodyDefinitions [{
      sphereCollisionRadius=0
      boundBone {boneName=dyn_hair_bun_r_02}
      boxExtents=<20 15 15>
      constraintSetup {angularLimitsMax=<1 15 5>, angularLimitsMin=<-1 -15 -15>, angularTarget=<1 0 0>}
      localJointOffset=<2 0 0>
     }]
    relativeSpaceBone {boneName=root}
   }
   n31 = Anim Dynamics - Bone: dyn_hair_bun_r_03(ComponentPose=@n30, Alpha=1, ExternalForce="0,0,0") {
    angularDampingOverride=0.98
    angularSpringConstant=50
    bAngularSpring=true
    bGravityOverrideInSimSpace=true
    bOverrideAngularDamping=true
    simulationSpace=RootRelative
    alphaBoolBlend {blendOption=Linear}
    boundBone {boneName=dyn_hair_bun_r_03}
    componentAppliedLinearAccClamp=<3500 3500 3500>
    componentLinearAccScale=<0.6 0.6 0.6>
    componentLinearVelScale=<0.6 0.6 0.6>
    physicsBodyDefinitions [{
      sphereCollisionRadius=0
      boundBone {boneName=dyn_hair_bun_r_03}
      boxExtents=<20 15 15>
      constraintSetup {angularLimitsMax=<1 15 5>, angularLimitsMin=<-1 -15 -15>, angularTarget=<1 0 0>}
      localJointOffset=<1 0 0>
     }]
    relativeSpaceBone {boneName=root}
   }
   n32 = Anim Dynamics - Bone: dyn_hair_bun_l_01(ComponentPose=@n31, Alpha=1, ExternalForce="0,0,0") {
    angularDampingOverride=0.98
    angularSpringConstant=50
    bAngularSpring=true
    bGravityOverrideInSimSpace=true
    bOverrideAngularDamping=true
    simulationSpace=RootRelative
    alphaBoolBlend {blendOption=Linear}
    boundBone {boneName=dyn_hair_bun_l_01}
    componentAppliedLinearAccClamp=<3500 3500 3500>
    componentLinearAccScale=<0.6 0.6 0.6>
    componentLinearVelScale=<0.6 0.6 0.6>
    physicsBodyDefinitions [{
      sphereCollisionRadius=0
      boundBone {boneName=dyn_hair_bun_l_01}
      boxExtents=<20 15 15>
      constraintSetup {angularLimitsMax=<1 15 5>, angularLimitsMin=<-1 -15 -15>, angularTarget=<1 0 0>}
      localJointOffset=<3 0 0>
     }]
    relativeSpaceBone {boneName=root}
   }
   n33 = Anim Dynamics - Bone: dyn_hair_bun_l_02(ComponentPose=@n32, Alpha=1, ExternalForce="0,0,0") {
    angularDampingOverride=0.98
    angularSpringConstant=50
    bAngularSpring=true
    bGravityOverrideInSimSpace=true
    bOverrideAngularDamping=true
    simulationSpace=RootRelative
    alphaBoolBlend {blendOption=Linear}
    boundBone {boneName=dyn_hair_bun_l_02}
    componentAppliedLinearAccClamp=<3500 3500 3500>
    componentLinearAccScale=<0.6 0.6 0.6>
    componentLinearVelScale=<0.6 0.6 0.6>
    physicsBodyDefinitions [{
      sphereCollisionRadius=0
      boundBone {boneName=dyn_hair_bun_l_02}
      boxExtents=<20 15 15>
      constraintSetup {angularLimitsMax=<1 15 5>, angularLimitsMin=<-1 -15 -15>, angularTarget=<1 0 0>}
      localJointOffset=<2 0 0>
     }]
    relativeSpaceBone {boneName=root}
   }
   n34 = Anim Dynamics - Bone: dyn_hair_bun_l_03(ComponentPose=@n33, Alpha=1, ExternalForce="0,0,0") {
    angularDampingOverride=0.98
    angularSpringConstant=50
    bAngularSpring=true
    bGravityOverrideInSimSpace=true
    bOverrideAngularDamping=true
    simulationSpace=RootRelative
    alphaBoolBlend {blendOption=Linear}
    boundBone {boneName=dyn_hair_bun_l_03}
    componentAppliedLinearAccClamp=<3500 3500 3500>
    componentLinearAccScale=<0.6 0.6 0.6>
    componentLinearVelScale=<0.6 0.6 0.6>
    physicsBodyDefinitions [{
      sphereCollisionRadius=0
      boundBone {boneName=dyn_hair_bun_l_03}
      boxExtents=<20 15 15>
      constraintSetup {angularLimitsMax=<1 15 5>, angularLimitsMin=<-1 -15 -15>, angularTarget=<1 0 0>}
      localJointOffset=<1 0 0>
     }]
    relativeSpaceBone {boneName=root}
   }
   n35 = Anim Dynamics - Bone: dyn_helmet_strap(ComponentPose=@n34, Alpha=0.75, ExternalForce="0,0,0") {
    angularDampingOverride=0.99
    angularSpringConstant=2500
    bAngularSpring=true
    bGravityOverrideInSimSpace=true
    bOverrideAngularDamping=true
    simulationSpace=RootRelative
    alphaBoolBlend {blendOption=Linear}
    boundBone {boneName=dyn_helmet_strap}
    componentAppliedLinearAccClamp=<3500 3500 3500>
    componentLinearAccScale=<0.6 0.6 0.6>
    componentLinearVelScale=<0.6 0.6 0.6>
    physicsBodyDefinitions [{
      sphereCollisionRadius=0
      boundBone {boneName=dyn_helmet_strap}
      boxExtents=<15 15 15>
      constraintSetup {angularLimitsMax=<1 3 5>, angularLimitsMin=<-1 -3 -35>, angularTarget=<1 0 0>}
      localJointOffset=<1 0 0>
     }]
    relativeSpaceBone {boneName=root}
   }
   n36 = "Constraint - dyn_Hair_Back_01 (PP)"(ComponentPose=@n35, Alpha=1, ConstraintWeights_0=0.6, ConstraintWeights_1=0.4) {
    alphaBoolBlend {blendOption=Linear}
    boneToModify {boneName=dyn_Hair_Back_01}
    constraintSetup [{targetBone {boneName=head}}, {targetBone {boneName=spine_05}}]
    constraintWeights [1,1]
   }
   n37 = "Constraint - dyn_Hair_Back_02 (PP)"(ComponentPose=@n36, Alpha=1, ConstraintWeights_0=0.4, ConstraintWeights_1=0.6) {
    alphaBoolBlend {blendOption=Linear}
    boneToModify {boneName=dyn_Hair_Back_02}
    constraintSetup [{targetBone {boneName=head}}, {targetBone {boneName=spine_05}}]
    constraintWeights [1,1]
   }
   n38 = "Constraint - dyn_Hair_Back_03 (PP)"(ComponentPose=@n37, Alpha=1, ConstraintWeights_0=0.6, ConstraintWeights_1=0.4) {
    alphaBoolBlend {blendOption=Linear}
    boneToModify {boneName=dyn_Hair_Back_03}
    constraintSetup [{targetBone {boneName=spine_05}}, {targetBone {boneName=spine_05}}]
    constraintWeights [1,1]
   }
   n39 = "Constraint - dyn_Hair_Back_04 (PP)"(ComponentPose=@n38, Alpha=1, ConstraintWeights_0=1, ConstraintWeights_1=0) {
    alphaBoolBlend {blendOption=Linear}
    boneToModify {boneName=dyn_Hair_Back_04}
    constraintSetup [{targetBone {boneName=spine_05}}, {targetBone {boneName=spine_05}}]
    constraintWeights [1,1]
   }
   n40 = "Constraint - dyn_Hair_BackL_01 (PP)"(ComponentPose=@n39, Alpha=1, ConstraintWeights_0=0.3, ConstraintWeights_1=0.7) {
    alphaBoolBlend {blendOption=Linear}
    boneToModify {boneName=dyn_Hair_BackL_01}
    constraintSetup [{targetBone {boneName=head}}, {targetBone {boneName=dyn_Hair_Back_01}}]
    constraintWeights [1,1]
   }
   n41 = "Constraint - dyn_Hair_BackL_02 (PP)"(ComponentPose=@n40, Alpha=1, ConstraintWeights_0=0.3, ConstraintWeights_1=0.7) {
    alphaBoolBlend {blendOption=Linear}
    boneToModify {boneName=dyn_Hair_BackL_02}
    constraintSetup [{targetBone {boneName=head}}, {targetBone {boneName=dyn_Hair_Back_02}}]
    constraintWeights [1,1]
   }
   n42 = "Constraint - dyn_Hair_BackL_03 (PP)"(ComponentPose=@n41, Alpha=1, ConstraintWeights_0=0.5, ConstraintWeights_1=0.5) {
    alphaBoolBlend {blendOption=Linear}
    boneToModify {boneName=dyn_Hair_BackL_03}
    constraintSetup [{targetBone {boneName=spine_05}}, {targetBone {boneName=dyn_Hair_Back_03}}]
    constraintWeights [1,1]
   }
   n43 = "Constraint - dyn_Hair_BackL_04 (PP)"(ComponentPose=@n42, Alpha=1, ConstraintWeights_0=0.8, ConstraintWeights_1=0.2) {
    alphaBoolBlend {blendOption=Linear}
    boneToModify {boneName=dyn_Hair_BackL_04}
    constraintSetup [{targetBone {boneName=spine_05}}, {targetBone {boneName=dyn_Hair_Back_04}}]
    constraintWeights [1,1]
   }
   n44 = "Constraint - dyn_Hair_BackR_01 (PP)"(ComponentPose=@n43, Alpha=1, ConstraintWeights_0=0.3, ConstraintWeights_1=0.7) {
    alphaBoolBlend {blendOption=Linear}
    boneToModify {boneName=dyn_Hair_BackR_01}
    constraintSetup [{targetBone {boneName=head}}, {targetBone {boneName=dyn_Hair_Back_01}}]
    constraintWeights [1,1]
   }
   n45 = "Constraint - dyn_Hair_BackR_02 (PP)"(ComponentPose=@n44, Alpha=1, ConstraintWeights_0=0.3, ConstraintWeights_1=0.7) {
    alphaBoolBlend {blendOption=Linear}
    boneToModify {boneName=dyn_Hair_BackR_02}
    constraintSetup [{targetBone {boneName=head}}, {targetBone {boneName=dyn_Hair_Back_02}}]
    constraintWeights [1,1]
   }
   n46 = "Constraint - dyn_Hair_BackR_03 (PP)"(ComponentPose=@n45, Alpha=1, ConstraintWeights_0=0.5, ConstraintWeights_1=0.5) {
    alphaBoolBlend {blendOption=Linear}
    boneToModify {boneName=dyn_Hair_BackR_03}
    constraintSetup [{targetBone {boneName=spine_05}}, {targetBone {boneName=dyn_Hair_Back_03}}]
    constraintWeights [1,1]
   }
   n47 = "Constraint - dyn_Hair_BackR_04 (PP)"(ComponentPose=@n46, Alpha=1, ConstraintWeights_0=0.8, ConstraintWeights_1=0.2) {
    alphaBoolBlend {blendOption=Linear}
    boneToModify {boneName=dyn_Hair_BackR_04}
    constraintSetup [{targetBone {boneName=spine_05}}, {targetBone {boneName=dyn_Hair_Back_04}}]
    constraintWeights [1,1]
   }
   Output Pose(Result=Component To Local(ComponentPose=@n47)) {layerGroup=None}
  ]}
 EventGraph {nodes [
   Event BlueprintUpdateAnimation()
   > n1 = Gravity Override State Function DEPRECATED(
    CharacterPartAnimInstance=Self-Reference()
    GameTimeInSeconds=GetGameTimeInSeconds()
    GravityOverrideParams=Get GravityOverrideParams() -> (GravityOverrideParams:Gravity Override Params Struct Structure)
   )
   > Set GravityOverride(GravityOverride=@n1)
   > n2 = Branch(Condition=CheckForBackbling())
   > Set ACC(ACC:Vector="0, 0, 0") -> (Output_Get:Vector)
   > Set VEL(VEL:Vector="0, 0, 0") -> (Output_Get:Vector)
   > Set backpack(backpack:double=0.5) -> (Output_Get:double)
   Set ACC(execute=@n2.else, ACC="0.3,0.3,0.5")
   > Set VEL(VEL="0.3,0.3,0.5")
   > Set backpack(backpack=0)
   n3 = Event OnBeginFalling()
   Set jumping(execute=[@n3.then, @n5.then], jumping:double=1) -> (Output_Get:double)
   n4 = Event OnEndFalling()
   Set jumping(execute=[@n4.then, @n6.then], jumping=0)
   n5 = Event OnBeginJumping()
   n6 = Event OnEndJumping()
  ]}
]
variables [
 ACC {type=Vector}
 backpack {type=number}
 GravityOverride {type=Vector}
 GravityOverrideParams {type=GravityOverrideParamsStruct}
 jumping {type=number}
 VEL {type=Vector}
]
